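# Translucent unlit smoke material for instanced meshes that animates billboarded puffs entirely from shader time.
# World position offset loops each sprite's size and travel with UV1 phase, adding sine drift and direction blending toward storm settings by weather intensity.
# Opacity combines a turbulence-distorted sphere mask with size cycle, distance and depth fades.
# Color supports contact tint, time-of-day, vertex-color hue variation and optional chromatic noise, with low-end fallbacks.
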Material /Game/Effects/Fort_Effects/Materials/Dust/M_Apollo_NoTick_Smoke_Slowest
graph {
 properties {
  bUsedWithInstancedStaticMeshes=true
  blendMode=BLEND_Translucent
  pixelDepthOffsetMode=PDOM_Legacy
  shadingModel=MSM_Unlit
  twoSided=true
  usedShadingModels=Unlit
 }
 nodes [
  n1 = Add(A=VertexColor().R, B=TextureCoordinate())
  n2 = ScalarParameter() {defaultValue=0.25, group=Chromatic, parameterName=TileSize_1}
  n3 = Time()
  n4 = ScalarParameter() {group=Chromatic, parameterName=Panning Scale}
  n5 = Constant2Vector() {g=-0.028}
  n6 = ScalarParameter() {defaultValue=1, group=Chromatic, parameterName=Velocity_Ctrl}
  n7 = TextureSampleParameter2D(UVs=Panner(Coordinate=Multiply(A=@n1, B=@n2), Time=Multiply(A=@n3, B=@n4), Speed=Multiply(A=@n5, B=@n6)) {speedY=-0.028}) {
   desc=Distort the Tiling Mask by this
   group=Chromatic
   parameterName=UVDistort_Normal_1
   samplerType=SAMPLERTYPE_Normal
   texture=/Game/Global/Textures/Default/Flowmaps/T_SwirlFlow_02_N
  }
  n8 = Multiply(A=ComponentMask(None=@n7.RGB) {g=true, r=true}, B=ScalarParameter() {defaultValue=0.08, group=Chromatic, parameterName=UVDistort})
  n9 = Add(A=@n8, B=Add(B=TextureCoordinate() {uTiling=0.5, vTiling=0.5}) {constA=-0.1})
  n10 = VertexColor()
  n11 = AppendVector(A=@n10.R, B=@n10.R)
  Add(A=@n9, B=@n11)
  n12 = WorldPosition()
  n13 = Multiply(A=ScalarParameter() {defaultValue=0.25, group=WPO, parameterName=MinSizeMult}, B=VertexColor().G)
  n14 = ScalarParameter() {defaultValue=500, group=WPO, parameterName=maxSize}
  n15 = Divide(A=Time() {bOverride_Period=true, period=16}) {constB=16}
  n16 = Divide(A=Time() {bOverride_Period=true, period=20}) {constB=20}
  n17 = TextureCoordinate() {coordinateIndex=1}
  n18 = Multiply(A=Subtract(A=ComponentMask(None=@n17) {g=true}) {constB=0.5}) {constB=-2}
  n19 = StaticSwitchParameter(True=Add(A=StaticSwitchParameter(True=@n15, False=@n16) {defaultValue=true, parameterName=UseFastSprites}, B=@n18), False=@n18) {defaultValue=true, parameterName=Continuous Up and down scaling?}
  n20 = Sine(None=Add(A=@n19) {constB=-0.25})
  n21 = Power(Base=LinearInterpolate(Alpha=ConstantBiasScale(None=@n20))) {constExponent=1}
  n22 = Multiply(
   A=LinearInterpolate(A=Multiply(A=@n13, B=@n14), B=@n14, Alpha=@n21)
   B=MaterialFunctionCall(Distance=ScalarParameter() {defaultValue=500, group=WPO, parameterName=NearCamShrinkDist}) {materialFunction=&Near_Camera_Fade}.Result
  )
  n23 = Multiply(A=Time(), B=ScalarParameter() {defaultValue=0.001, group=WPO, parameterName=VarianceSpeed}) {constB=0.05}
  n24 = ComponentMask(None=TextureCoordinate() {coordinateIndex=1}) {r=true}
  n25 = Add(
   A=Multiply(A=Time(), B=ScalarParameter() {defaultValue=0.4, group=WPO, parameterName=DriftSpeed})
   B=Sine(None=Add(A=@n23, B=@n24)) {period=0.17}
  )
  n26 = AppendVector(A=Sine(None=@n25) {period=0.56}, B=Sine(None=@n25) {period=0.7})
  n27 = Sine(None=@n25) {period=0.82}
  n28 = ScalarParameter() {defaultValue=1, group=WPO, parameterName=MinHorizontalDrift}
  n29 = ScalarParameter() {defaultValue=1, group=WPO, parameterName=MaxHorizontalDrift}
  n30 = Sine(None=Add(A=@n23, B=@n24)) {period=0.14}
  n31 = Multiply(A=ObjectRadius(), B=Constant() {r=1}) {constB=0.5}
  n32 = LinearInterpolate(A=Multiply(A=@n31) {constB=-1}, B=@n31, Alpha=Frac(None=@n19))
  n33 = ScalarParameter() {defaultValue=1, group=WPO, parameterName=OverallSpeed}
  n34 = VectorParameter() {parameterName=OverallDirection, defaultValue=<rgba 1 0 0 1>}
  n35 = VectorParameter() {parameterName=OverallDirection_Storm, defaultValue=<rgba 1 0 0 1>}
  n36 = CollectionParameter() {
   collection=&FortniteMaterialParameters
   parameterName=CurrentWeatherEventIntensity
  }
  n37 = Multiply(A=Clamp(None=@n36), B=ScalarParameter() {parameterName=UseSnowStormSettings?})
  n38 = Add(
   A=Multiply(A=AppendVector(A=@n26, B=@n27), B=LinearInterpolate(A=@n28, B=@n29, Alpha=@n30))
   B=Multiply(A=Multiply(A=@n32, B=@n33), B=LinearInterpolate(A=@n34.RGB, B=@n35.RGB, Alpha=@n37))
  )
  n39 = Transform(None=@n38)
  Add(
   A=Add(
    A=MaterialFunctionCall!(Center Location=@n12.XYZ, 0-1 UVs, XY scale=@n22, OverridePivot, "Normals (Optional)") {materialFunction=&Sprite}.World position offset
    B=StaticSwitchParameter(True=@n39, False=@n38) {parameterName=UseWorldSpaceMovement?}
   )
   B=MaterialFunctionCall(Offset Amount=ScalarParameter() {group=WPO, parameterName="Offset Amount (S)"}) {materialFunction=/Engine/Functions/Engine_MaterialFunctions02/WorldPositionOffset/CameraOffset}.Clamped Camera Offset
  )
  n40 = Add(A=@n8, B=Add(B=TextureCoordinate() {uTiling=0.5, vTiling=0.5}) {constA=0.1})
  n41 = Multiply(A=Time(), B=ScalarParameter() {group=Chromatic, parameterName=Panning Scale})
  n42 = Multiply(A=Constant2Vector() {g=-0.0125}, B=@n6)
  n43 = TextureSampleParameter2D(UVs=Panner(Coordinate=Add(A=@n40, B=@n11), Time=@n41, Speed=@n42) {speedY=-0.0125}) {
   desc=UV1 Tiling Noise Mask
   group=Chromatic
   parameterName=TileAlphaMask_2
   samplerType=SAMPLERTYPE_LinearGrayscale
   texture=/Game/Global/Textures/Default/Noise/T_SwirlNoise_N_height
  }
  n44 = TextureSampleParameter2D(UVs=Panner(Coordinate=@n9, Time=@n41, Speed=@n42) {speedY=-0.0125}) {
   desc=UV1 Tiling Noise Mask
   group=Chromatic
   parameterName=TileAlphaMask_1
   samplerType=SAMPLERTYPE_LinearGrayscale
   texture=&T_SwirlNoise_N_height
  }
  n45 = AppendVector(A=@n43.R, B=@n44.R)
  n46 = Add(A=@n8, B=TextureCoordinate() {uTiling=0.5, vTiling=0.5})
  n47 = TextureSampleParameter2D(UVs=Panner(Coordinate=Add(A=@n46, B=@n11), Time=@n41, Speed=@n42) {speedY=-0.0125}) {
   desc=UV1 Tiling Noise Mask
   group=Chromatic
   parameterName=TileAlphaMask_3
   samplerType=SAMPLERTYPE_LinearGrayscale
   texture=&T_SwirlNoise_N_height
  }
  n48 = ScalarParameter() {defaultValue=15, group=Chromatic, parameterName=Chromatic_Mult}
  n49 = Multiply(A=Desaturation(None=AppendVector(A=@n45, B=@n47.B), Fraction=@n48), B=@n48)
  n50 = VertexColor()
  n51 = AppendVector(A=@n50.R, B=@n50.R)
  n52 = TextureCoordinate()
  n53 = ComponentMask(None=VectorParameter() {group=ColorTexture, parameterName=ColorTexture_UVScale, defaultValue=<rgba 1 1 0 0>}.RGB) {g=true, r=true}
  n54 = VertexColor()
  n55 = AppendVector(A=@n54.R, B=@n54.R)
  n56 = ScalarParameter() {defaultValue=0.1, group=Texture values, parameterName=SpeedMult_02}
  n57 = TextureCoordinate()
  n58 = AppendVector(
   A=ScalarParameter() {defaultValue=1, group=Texture values, parameterName=TileX}
   B=ScalarParameter() {defaultValue=1, group=Texture values, parameterName=TileY}
  )
  n59 = TextureSampleParameter2D(UVs=Rotator(Coordinate=Multiply(A=@n57, B=@n58)) {centerX=1, speed=0.0312}) {
   parameterName=DistortionTexture
   samplerType=SAMPLERTYPE_LinearColor
   texture=/Game/Global/Textures/Default/Tileable_Patterns/T_Tile_Smoke_Small
  }
  n60 = TextureCoordinate()
  n61 = TextureSampleParameter2D(UVs=Rotator(Coordinate=Multiply(A=@n60, B=@n58)) {centerY=1, speed=-0.0312}) {
   parameterName=DistortionTexture
   samplerType=SAMPLERTYPE_LinearColor
   texture=&T_Tile_Smoke_Small
  }
  n62 = Multiply(
   A=Add(A=ComponentMask(None=@n59.RGB) {b=true, g=true, r=true}, B=ComponentMask(None=@n61.RGB) {b=true, g=true, r=true})
   B=ScalarParameter() {defaultValue=0.1, group=Texture values, parameterName=BaseDistort}
  )
  n63 = Constant()
  n64 = ShadingPathSwitch(Default=QualitySwitch(Default=ComponentMask(None=@n62) {g=true, r=true}, Low=@n63), Mobile=@n63) {desc=Migrated from FeatureLevelSwitch}
  n65 = Add(A=Panner(Coordinate=Multiply(A=@n52, B=@n53), Speed=Multiply(A=@n55, B=@n56)) {speedX=-0.02, speedY=0.1}, B=@n64)
  n66 = StaticComponentMaskParameter(None=TextureSampleParameter2D(UVs=Add(A=@n51, B=@n65)) {group=ColorTexture, parameterName=ColorTexture, texture=/Game/Global/Textures/Default/Caustics/T_Turbulence_Seamless}.RGBA) {defaultG=true, group=ColorTexture, parameterName=ColorTextureMask}
  n67 = DepthFade(FadeDistance=ScalarParameter() {defaultValue=50, parameterName=ShadowColor_FadeDistance}) {fadeDistanceDefault=140}
  n68 = VectorParameter() {parameterName=BaseColor, defaultValue=<rgba 0.4375 0.3488 0.2871 0>}
  n69 = Multiply(A=@n68.RGB, B=ScalarParameter() {defaultValue=0.25, parameterName=LowendColorMult})
  n70 = StaticSwitchParameter(True=QualitySwitch(Default=ShadingPathSwitch(Default=@n68.RGB, Forward=@n69, Mobile=@n69), Low=@n69), False=@n68.RGB) {parameterName=UseLowEndOpacityCheck?}
  n71 = QualitySwitch(
   Default=LinearInterpolate(
    A=VectorParameter() {parameterName=ContactColor, defaultValue=<rgba 0.5932 0.9323 0.3842 0>}.RGB
    B=VectorParameter() {parameterName=BaseColor, defaultValue=<rgba 0.4375 0.3488 0.2871 0>}.RGB
    Alpha=StaticSwitchParameter(True=@n66, False=@n67) {group=ColorTexture, parameterName=Use Texture for Color}
   )
   Low=@n70
   Medium=@n70
  )
  n72 = MaterialFunctionCall(
   Color_Morning=VectorParameter() {group=TOD, parameterName="Color_Morning (V3)", defaultValue=<rgba 1 1 1 0>}.RGB
   Color_Day=VectorParameter() {group=TOD, parameterName="Color_Day (V3)", defaultValue=<rgba 1 1 1 0>}.RGB
   Color_Evening=VectorParameter() {group=TOD, parameterName="Color_Evening (V3)", defaultValue=<rgba 1 1 1 0>}.RGB
   Color_Night=VectorParameter() {group=TOD, parameterName="Color_Night (V3)", defaultValue=<rgba 1 1 1 0>}.RGB
  ) {materialFunction=&Color_Time_of_Day}
  n73 = MaterialFunctionCall(
   Hue Shift Percentage=Multiply(A=VertexColor().B) {constB=0.05}
   Texture=StaticSwitchParameter(True=Multiply(A=@n71, B=@n72.Result), False=@n72.Result) {parameterName=Use Color TOD?}
  ) {materialFunction=&HueShift}
  MaterialFunctionCall(Emissive=StaticSwitchParameter(True=Multiply(A=@n49, B=@n73.Result), False=@n73.Result) {parameterName=UseChromatic?}) {materialFunction=&MF_FX_EmissiveAdjustments}
  n74 = MaterialFunctionCall!(
   NearFadeLength=ScalarParameter() {defaultValue=15000, group=Scaler, parameterName="NearFadeLength (S)"}
   NearFadeOffset=ScalarParameter() {defaultValue=10000, group=Scaler, parameterName="NearFadeOffset (S)"}
   NearFadeCurve
   NearFadeStartingScale
   NearFadeEndingScale
   FarFadeLength=ScalarParameter() {defaultValue=15000, group=Scaler, parameterName="FarFadeLength (S)"}
   FarFadeOffset=ScalarParameter() {defaultValue=75000, group=Scaler, parameterName="FarFadeOffset (S)"}
   FarFadeCurve
   FarFadeEndingScale
   Opacity Near Distance Bias=ScalarParameter() {group=Scaler, parameterName="Opacity Near Distance Bias Opacity (S)"}
   Opacity Far Distance Bias=ScalarParameter() {group=Scaler, parameterName="Opacity Far Distance Bias Opacity (S)"}
  ) {materialFunction=/Game/Effects/Fort_Effects/Materials/_Functions/MF_SmoothLod}
  n75 = ScalarParameter() {defaultValue=4, group=Opacity, parameterName=OpactiyDistScalePower}
  n76 = VertexColor()
  n77 = AppendVector(A=@n76.R, B=@n76.R)
  n78 = TextureCoordinate()
  n79 = VertexColor()
  n80 = AppendVector(A=@n79.R, B=@n79.R)
  n81 = ScalarParameter() {defaultValue=0.1, group=Texture values, parameterName=SpeedMult}
  n82 = Add(A=Panner(Coordinate=Multiply(A=@n78, B=@n58), Speed=Multiply(A=@n80, B=@n81)) {speedX=0.02, speedY=0.15}, B=@n64)
  n83 = VertexColor()
  n84 = AppendVector(A=@n83.R, B=@n83.R)
  n85 = ScalarParameter() {defaultValue=0.1, group=Texture values, parameterName=SpeedMult_02}
  n86 = Add(A=Panner(Coordinate=Multiply(A=@n78, B=@n58), Speed=Multiply(A=@n80, B=@n85)) {speedX=-0.02, speedY=0.1}, B=@n64)
  n87 = Add(
   A=TextureSampleParameter2D(UVs=Add(A=@n77, B=@n82)) {parameterName=SmokeTexture, texture=&T_Turbulence_Seamless}.G
   B=TextureSampleParameter2D(UVs=Add(A=@n84, B=@n86)) {parameterName=SmokeTexture, texture=&T_Turbulence_Seamless}.G
  )
  n88 = ScalarParameter() {defaultValue=0.15, parameterName=Contrast}
  n89 = SphereMask(
   A=TextureCoordinate()
   B=Constant2Vector() {g=0.5, r=0.5}
   Radius=Multiply(A=ScalarParameter() {defaultValue=0.35, parameterName=EdgeSoften}, B=Power(Base=@n87, Exp=@n88))
   Hardness=ScalarParameter() {parameterName=Hardness}
  )
  n90 = LinearInterpolate(Alpha=VertexColor().B) {constA=0.5}
  n91 = ScalarParameter() {defaultValue=1, parameterName=OpacityMult}
  n92 = ScalarParameter() {defaultValue=1, parameterName=OpacityMult_LowEnd}
  n93 = ShadingPathSwitch(Default=@n91, Forward=@n92, Mobile=@n92)
  n94 = Multiply(
   A=Multiply(A=Power(Base=@n74.Opacity, Exp=@n75), B=Multiply(A=@n89, B=@n90))
   B=StaticSwitchParameter(True=QualitySwitch(Default=@n93, Low=@n92), False=@n91) {parameterName=UseLowEndOpacityCheck?}
  )
  n95 = Multiply(A=@n94, B=DepthFade(FadeDistance=ScalarParameter() {defaultValue=150, group=Opacity, parameterName=Opacity_Depth_FadeDistance}))
  n96 = ScalarParameter() {defaultValue=1, group=Opacity, parameterName="Value_Morning (S)"}
  n97 = ScalarParameter() {defaultValue=1, group=Opacity, parameterName="Value_Day (S)"}
  n98 = ScalarParameter() {defaultValue=1, group=Opacity, parameterName="Value_Evening (S)"}
  n99 = ScalarParameter() {defaultValue=1, group=Opacity, parameterName="Value_Night (S)"}
  Multiply(
   A=@n21
   B=Multiply(
    A=QualitySwitch(Default=@n95, Low=@n94)
    B=MaterialFunctionCall(Value_Morning=@n96, Value_Day=@n97, Value_Evening=@n98, Value_Night=@n99) {materialFunction=&Scalar_Time_of_Day}.Result
   )
  )
  ScalarParameter() {defaultValue=1, parameterName=Size}
 ]
}
