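# Translucent unlit material for the glowing edge of liquid in the Apollo power plant effect.
# Two panning normal maps distort scrolling caustics, sampled three times with Dynamic Parameter-driven offsets for color fringing; mobile and low quality use one sample.
# The caustics, weighted along U, blend two glow colors, boosted to HDR.
# A U-gradient power shapes intensity and opacity, with depth fade softening intersections and a camera offset shifting the mesh.
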Material /Game/Effects/Fort_Effects/Materials/Dust/M_Apollo_Powerplant_LiquidEdgeGlow
graph {
 properties {blendMode=BLEND_Translucent, pixelDepthOffsetMode=PDOM_Legacy, shadingModel=MSM_Unlit, usedShadingModels=Unlit}
 nodes [
  MaterialFunctionCall(Offset Amount=ScalarParameter() {defaultValue=5, parameterName="Offset Amount (S)"}) {materialFunction=/Engine/Functions/Engine_MaterialFunctions02/WorldPositionOffset/CameraOffset}
  n1 = VectorParameter() {parameterName=LightColor, defaultValue=<rgba 0.2861 0.5338 0.6 0>}
  n2 = ComponentMask(None=VectorParameter() {parameterName=Tiling_Multi, defaultValue=<rgba 1 1 0 0>}.RGB) {g=true, r=true}
  n3 = TextureCoordinate() {vTiling=25}
  n4 = Multiply(
   A=ComponentMask(None=VectorParameter() {parameterName=Tiling_Multi, defaultValue=<rgba 1 1 0 0>}.RGB) {g=true, r=true}
   B=TextureCoordinate() {unMirrorU=true, vTiling=4}
  )
  n5 = Panner(Coordinate=@n4) {speedX=0.1, speedY=0.1}
  n6 = Add(
   A=TextureSample(UVs=Panner(Coordinate=@n4) {speedX=0.15, speedY=0.15}) {samplerType=SAMPLERTYPE_Normal, texture=&T_Weap_Distort_N_02}.RGB
   B=Divide(A=TextureSample(UVs=@n5) {samplerType=SAMPLERTYPE_Normal, texture=/Game/Effects/Fort_Effects/Gameplay/Water/Textures/T_WaterRipples_01}.RGB)
  )
  n7 = Panner(Coordinate=Add(A=Multiply(A=@n2, B=@n3), B=ComponentMask(None=@n6) {g=true, r=true})) {speedX=0.07, speedY=0.07}
  n8 = TextureSample(UVs=@n7) {texture=&T_ShieldCaustics}
  n9 = DynamicParameter() {defaultValue=<rgba 0 0.3 -0.3 1>, paramNames [CloseUp,OffsetA,OffsetB,Param4]}
  n10 = TextureCoordinate() {uTiling=1.5}
  n11 = ComponentMask(None=VectorParameter() {parameterName=Tiling_Multi, defaultValue=<rgba 1 1 0 0>}.RGB) {g=true, r=true}
  n12 = TextureSample(UVs=Panner(Coordinate=Multiply(A=@n10, B=@n11)) {speedX=0.15}) {samplerType=SAMPLERTYPE_Normal, texture=&T_Weap_Distort_N_02}
  n13 = Saturate(None=OneMinus(None=MaterialFunctionCall() {materialFunction=&LinearGradient}.VGradient))
  n14 = LinearInterpolate(B=@n9.OffsetA, Alpha=LinearInterpolate(B=@n12.R, Alpha=@n13)) {constB=0.3}
  n15 = AppendVector(A=TextureSample(UVs=Add(A=@n7, B=@n14) {constB=0.015}) {texture=&T_ShieldCaustics}.R, B=@n8.R)
  n16 = LinearInterpolate(B=@n12.G, Alpha=@n13)
  n17 = TextureSample(UVs=Add(A=@n7, B=LinearInterpolate(B=@n9.OffsetB, Alpha=@n16) {constB=-0.3}) {constB=-0.015}) {texture=&T_ShieldCaustics}
  n18 = Desaturation(None=SquareRoot(None=AppendVector(A=@n15, B=@n17.R)), Fraction=Constant() {r=-0.3})
  n19 = LinearInterpolate(
   A=LinearInterpolate(A=VectorParameter() {parameterName=DarkColor, defaultValue=<rgba 0.0324 0.0686 0.0729 0>}.RGB, B=@n1.RGB, Alpha=@n8.R)
   B=@n1.RGB
   Alpha=QualitySwitch(Default=ShadingPathSwitch(Default=@n18, Mobile=@n8.R) {desc=Migrated from FeatureLevelSwitch}, Low=@n8.R)
  )
  n20 = ComponentMask(None=TextureCoordinate()) {r=true}
  n21 = Power(Base=@n20)
  MaterialFunctionCall(
   LDR Input=LinearInterpolate(
    A=VectorParameter() {parameterName=GlowColor, defaultValue=<rgba 0.3289 0 1 0>}.RGB
    B=VectorParameter() {parameterName=GlowColor_Bottom, defaultValue=<rgba 0.0072 0 1 0>}.RGB
    Alpha=Multiply(A=@n19, B=@n20)
   )
   HDR Intensity Multiplier=Multiply(A=ScalarParameter() {defaultValue=5, parameterName="HDR Intensity Multiplier (S)"}, B=@n21)
  ) {materialFunction=&DeriveHDRfromLDR}
  n22 = ScalarParameter() {defaultValue=25, parameterName=FadeDistance}
  Multiply(A=ScalarParameter() {defaultValue=1, parameterName=Opacity}, B=Multiply(A=Multiply(A=@n21), B=DepthFade(FadeDistance=@n22)))
 ]
}
