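# Translucent Default Lit dust and smoke material for Niagara sprites, using an 8x8 smoke SubUV flipbook and spherical particle normals.
# Dynamic parameters erode and soften texture alpha, which is scaled by particle alpha and feeds depth fade and temporal dither nodes.
# Color blends from tinted texture to flat particle color and feeds emissive adjustment functions.
# A normal is derived from the smoke texture as a heightmap, flattened by a dynamic parameter and transformed.
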Material /Game/Effects/Fort_Effects/Materials/Dust/M_Dust_Niagara_Smoke_Transparent_Lit
graph {
 properties {
  bGenerateSphericalParticleNormals=true
  bUsedWithInstancedStaticMeshes=true
  bUsedWithNiagaraSprites=true
  blendMode=BLEND_Translucent
  pixelDepthOffsetMode=PDOM_Legacy
  usedShadingModels=Default Lit
 }
 nodes [
  n1 = CollectionParameter() {
   collection=&FortniteMaterialParameters
   parameterName=SunAndMoonModelDirectionalVector
  }
  Clamp(None=DotProduct(A=ComponentMask(None=@n1) {b=true, g=true, r=true}, B=ReflectionVectorWS()))
  Constant()
  Constant()
  Constant() {r=1}
  n2 = ParticleSubUV!(UVs, Tex, Apply View MipBias) {texture=/Game/Effects/Fort_Effects/Textures/Smoke/T_Smoke_Thick_8x8_02}
  n3 = DynamicParameter() {defaultValue=<rgba 0 0 1 1>, paramNames [Erosion, Erosion_Softness, DepthFade, Emmisive]}
  n4 = Subtract(A=FeatureLevelSwitch(Default=QualitySwitch(Default=@n2.A)), B=@n3.Erosion)
  n5 = Multiply(A=Saturate(None=Divide(A=@n4, B=@n3.Erosion_Softness)), B=ParticleColor().A)
  DepthFade(Opacity=@n5) {fadeDistanceDefault=500}
  n6 = DynamicParameter() {parameterIndex=1, defaultValue=<rgba 1 1 0 1>, paramNames [NormalFlat,RandomDither,BlendAmount,Param4]}
  MaterialFunctionCall(Alpha Threshold=@n5, Offset=Constant2Vector() {g=0.5, r=0.5}, Random=@n6.RandomDither) {materialFunction=&DitherTemporalAA_Fortnite}
  n7 = ParticleColor()
  n8 = DynamicParameter() {parameterIndex=1, defaultValue=<rgba 1 1 0 1>, paramNames [NormalFlat,RandomDither,BlendAmount,Param4]}
  n9 = LinearInterpolate(A=Multiply(A=@n2.RGB, B=@n7.RGB), B=@n7.RGB, Alpha=@n8.BlendAmount)
  MaterialFunctionCall(Emissive=Multiply(A=@n9, B=DynamicParameter() {defaultValue=<rgba 0 0 1 1>, paramNames [Erosion, Erosion_Softness, DepthFade, Emmisive]}.Emmisive)) {materialFunction=&MF_FX_EmissiveAdjustments}
  MaterialFunctionCall(Emissive=@n9) {materialFunction=&MF_FX_EmissiveAdjustments}
  Multiply(A=@n2.A)
  Multiply(
   A=MaterialFunctionCall(Exponent=ScalarParameter() {defaultValue=250, parameterName=Expo}) {materialFunction=&AxisAlignedFresnel}.Fresnel
   B=ScalarParameter() {defaultValue=1, parameterName=1}
  )
  Multiply(
   A=MaterialFunctionCall(Alpha Threshold=ScalarParameter() {defaultValue=1, parameterName=Depth Offset}, Random=Constant() {r=1}) {materialFunction=&DitherTemporalAA}.Result
   B=ScalarParameter() {defaultValue=25000, parameterName=Pixel Offset}
  )
  Multiply(
   A=@n2.A
   B=DepthFade(FadeDistance=DynamicParameter() {defaultValue=<rgba 0 0 1 1>, paramNames [Erosion, Erosion_Softness, DepthFade, Emmisive]}.DepthFade)
  )
  ParticleColor()
  ScalarParameter() {parameterName=Flat}
  ScalarParameter() {group=Texture, parameterName=ColorIntensity_SunSide_Mult}
  ScalarParameter() {defaultValue=5, group=Texture, parameterName=ColorIntensity_ShadowSide_Mult}
  ScalarParameter() {parameterName=Blend}
  n10 = TextureObject() {texture=&T_Smoke_Thick_8x8_02}
  Transform(None=MaterialFunctionCall(
    Normal=MaterialFunctionCall!(Height Map=@n10, Normal Map Intensity, Height Map UV Offset, Coordinates, Height Map Channel Selector) {materialFunction=/Engine/Functions/Engine_MaterialFunctions03/Procedurals/NormalFromHeightmap}.Result
    Flatness=@n6.NormalFlat
   ) {materialFunction=&FlattenNormal}.Result)
 ]
}
