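# Translucent, unlit dust and smoke material that animates looping billboard puffs from shader time, with instanced static mesh support.
# World position offset moves sprites across the object bounds on a UV1-staggered loop, with sine-driven size, near-camera shrink and optional landscape conforming via height RVT.
# An 8x8 motion-vector flipbook feeds time-of-day-tinted emissive and opacity shaped by loop, distance and depth fades plus erosion.
# Vertex color varies size, playback speed and opacity.
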Material /Game/Effects/Fort_Effects/Materials/Dust/M_Dust_NoTick_Smoke
graph {
 properties {
  bGenerateSphericalParticleNormals=true
  bUseTranslucencyVertexFog=false
  bUsedWithInstancedStaticMeshes=true
  blendMode=BLEND_Translucent
  pixelDepthOffsetMode=PDOM_Legacy
  shadingModel=MSM_Unlit
  twoSided=true
  usedShadingModels=Unlit
 }
 nodes [
  n1 = WorldPosition()
  n2 = Multiply(A=ScalarParameter() {defaultValue=0.25, group=3. WPO, parameterName=MinSizeMult}, B=VertexColor().G)
  n3 = ScalarParameter() {defaultValue=500, group=3. WPO, parameterName=MaxSize}
  n4 = Divide(A=Add(A=Time(), B=ScalarParameter() {parameterName=TimeOffset}), B=ScalarParameter() {defaultValue=5, parameterName=LoopDuration}) {constB=5}
  n5 = TextureCoordinate() {coordinateIndex=1}
  n6 = Multiply(A=Subtract(A=ComponentMask(None=@n5) {g=true}) {constB=0.5}) {constB=-2}
  n7 = StaticSwitchParameter(True=Add(A=Frac(None=@n4), B=@n6), False=@n6) {defaultValue=true, parameterName=Continuous Up and down scaling?}
  n8 = Add(A=@n7) {constB=-0.25}
  n9 = LinearInterpolate(Alpha=ConstantBiasScale(None=Sine(None=@n8)))
  n10 = Multiply(
   A=LinearInterpolate(A=Multiply(A=@n2, B=@n3), B=@n3, Alpha=Power(Base=@n9) {constExponent=1})
   B=MaterialFunctionCall(Distance=ScalarParameter() {defaultValue=500, group=3. WPO, parameterName=NearCamShrinkDist}) {materialFunction=&Near_Camera_Fade}.Result
  )
  n11 = Multiply(A=ObjectBounds(), B=ScalarParameter() {defaultValue=1, parameterName=ObjectRadius}) {constB=0.5}
  n12 = Multiply(A=@n11) {constB=-1}
  n13 = Frac(None=@n7)
  n14 = Multiply(
   A=Multiply(A=LinearInterpolate(A=@n12, B=@n11, Alpha=@n13), B=ScalarParameter() {defaultValue=1, group=WPO, parameterName=OverallSpeed})
   B=VectorParameter() {parameterName=OverallDirection, defaultValue=<rgba 1 0 0 1>}.RGB
  )
  n15 = WorldPosition()
  n16 = ComponentMask(None=ObjectRadius()) {b=true}
  n17 = Add(
   A=Subtract(
    A=RuntimeVirtualTextureSample!(UVs, World Position=@n15.XYZ) {materialType=WorldHeight, virtualTexture=/Game/Athena/Apollo/Maps/Landscape/VT/RVT_Apollo_Landscape_Height}.WorldHeight
    B=ComponentMask(None=@n15.XYZ) {b=true}
   )
   B=@n16
  )
  n18 = Multiply(A=@n16, B=ScalarParameter() {defaultValue=1, group=3. WPO, parameterName=LandScapeOffsetHeight})
  n19 = ComponentMask(None=Multiply(A=VectorParameter() {parameterName=ConformMult, defaultValue=<rgba 0 0 1 1>}.RGB, B=Subtract(A=@n17, B=@n18))) {b=true}
  n20 = Constant()
  n21 = AppendVector(
   A=ComponentMask(None=@n14) {g=true, r=true}
   B=Add(A=ComponentMask(None=@n14) {b=true}, B=StaticSwitchParameter(True=@n19, False=@n20) {parameterName=UseSpawnonLandscape?})
  )
  n22 = Transform(None=@n21)
  Add(
   A=Add(
    A=MaterialFunctionCall!(Center Location=@n1.XYZ, 0-1 UVs, XY scale=@n10, OverridePivot, "Normals (Optional)") {materialFunction=&Sprite}.World position offset
    B=StaticSwitchParameter(True=@n22, False=@n21) {parameterName=UseWorldSpaceMovement?}
   )
   B=MaterialFunctionCall(Offset Amount=ScalarParameter() {group=3. WPO, parameterName=CameraOffsetAmount}) {materialFunction=/Engine/Functions/Engine_MaterialFunctions02/WorldPositionOffset/CameraOffset}.Clamped Camera Offset
  )
  CollectionParameter() {
   collection=&FortniteMaterialParameters
   parameterName=SunlightDirectionalVector
  }
  ComponentMask() {g=true, r=true}
  n23 = ScalarParameter() {parameterName=BrightnessMin}
  n24 = NamedRerouteUsage()
  n25 = VectorParameter() {parameterName=MainColor, defaultValue=<rgba 1 1 1 1>}
  n26 = LinearInterpolate(
   A=@n25.RGB
   B=MaterialFunctionCall() {materialFunction=/Game/Effects/Fort_Effects/Materials/_Functions/CurveColor_Time_of_Day}.Result
   Alpha=ScalarParameter() {group=1. TextureValues, parameterName=TODAmount}
  )
  n27 = Multiply(
   A=Multiply(A=ScalarParameter() {defaultValue=1, parameterName=EmissiveMult}, B=LinearInterpolate(A=@n23, Alpha=@n24))
   B=VertexInterpolator(VS=StaticSwitchParameter(True=@n26, False=@n25.RGB) {defaultValue=true, parameterName=UseTOD?}).PS
  )
  MaterialFunctionCall(Emissive=@n27) {materialFunction=&MF_FX_EmissiveAdjustments}
  n28 = CollectionParameter() {
   collection=&FortniteMaterialParameters
   parameterName=SunlightDirectionalVector
  }
  MaterialExpressionDotProduct_2 = DotProduct(A=CameraVectorWS(), B=ComponentMask(None=@n28) {b=true, g=true, r=true})
  MaterialExpressionScalarParameter_16 = ScalarParameter() {defaultValue=0.4, parameterName=Anisotropy}
  n29 = Custom!(CosTheta=@MaterialExpressionDotProduct_2, k=@MaterialExpressionScalarParameter_16) {
   code="\tfloat Inner = (1 + k * CosTheta);\r\n\treturn (1 - k * k) / (4 * PI * Inner * Inner);\r\n\r\n"
   description=Schlick Phase
   inputs [
    {inputName=CosTheta, input {expression=&self:MaterialExpressionDotProduct_2}}
    {inputName=k, input {expression=&self:MaterialExpressionScalarParameter_16}}
   ]
  }
  Multiply(A=OneMinus(None=@n29), B=@n27)
  n30 = ScalarParameter() {defaultValue=15000, group=2. Opacity, parameterName="NearFadeLength (S)"}
  n31 = ScalarParameter() {defaultValue=10000, group=2. Opacity, parameterName="NearFadeOffset (S)"}
  n32 = ScalarParameter() {defaultValue=15000, group=2. Opacity, parameterName="FarFadeLength (S)"}
  n33 = ScalarParameter() {defaultValue=75000, group=2. Opacity, parameterName="FarFadeOffset (S)"}
  n34 = NamedRerouteUsage()
  n35 = ShadingPathSwitch(Default=QualitySwitch(Default=@n34, Low=@n34), Mobile=@n34) {desc=Migrated from FeatureLevelSwitch}
  n36 = VertexColor()
  n37 = Multiply(
   A=VertexInterpolator(VS=MaterialFunctionCall!(
     NearFadeLength=@n30
     NearFadeOffset=@n31
     NearFadeCurve
     NearFadeStartingScale
     NearFadeEndingScale
     FarFadeLength=@n32
     FarFadeOffset=@n33
     FarFadeCurve
     FarFadeEndingScale
     Opacity Near Distance Bias
     Opacity Far Distance Bias
    ) {materialFunction=/Game/Effects/Fort_Effects/Materials/_Functions/MF_SmoothLod}.Opacity).PS
   B=Multiply(A=ShadingPathSwitch(Default=@n35, Forward=@n34), B=LinearInterpolate(Alpha=@n36.B) {constA=0.5})
  )
  n38 = DepthFade(FadeDistance=ScalarParameter() {defaultValue=150, group=2. Opacity, parameterName=Opacity_Depth_FadeDistance})
  n39 = ShadingPathSwitch(Default=QualitySwitch(Default=Multiply(A=@n37, B=@n38), Low=@n37), Mobile=@n37) {desc=Migrated from FeatureLevelSwitch}
  n40 = VertexInterpolator(VS=MaterialFunctionCall(
    Value_Morning=ScalarParameter() {defaultValue=1, group=2. Opacity, parameterName="Value_Morning (S)"}
    Value_Day=ScalarParameter() {defaultValue=1, group=2. Opacity, parameterName="Value_Day (S)"}
    Value_Evening=ScalarParameter() {defaultValue=1, group=2. Opacity, parameterName="Value_Evening (S)"}
    Value_Night=ScalarParameter() {defaultValue=1, group=2. Opacity, parameterName="Value_Night (S)"}
   ) {materialFunction=&Scalar_Time_of_Day}.Result)
  n41 = Subtract(A=Multiply(A=@n9, B=Multiply(A=@n39, B=@n40.PS)), B=ScalarParameter() {parameterName=Erosion})
  n42 = ScalarParameter() {parameterName=Erosion_Softness}
  Multiply(A=Saturate(None=Divide(A=@n41, B=@n42)), B=ScalarParameter() {defaultValue=1, group=2. Opacity, parameterName=OpacityMult})
  n43 = Time()
  n44 = LinearInterpolate(
   A=ScalarParameter() {defaultValue=1, group=1. TextureValues, parameterName=AnimationSpeed}
   B=ScalarParameter() {defaultValue=1, group=1. TextureValues, parameterName=AnimationSpeedMax}
   Alpha=VertexColor().G
  )
  n45 = MaterialFunctionCall!(
   "Animation  Phase (0-1)"=Frac(None=Multiply(A=@n43, B=@n44) {constB=0.1})
   Number of Columns X=Constant() {r=8}
   Number of Rows Y=Constant() {r=8}
   Texture=TextureObject() {texture=/Game/Effects/Fort_Effects/Textures/Smoke/T_Smoke_Thick_8x8_02}
   UVs
   Motion Vector Texture=TextureObject() {texture=/Game/Effects/Fort_Effects/Textures/Smoke/T_Smoke_Thick_MotionVector_8x8_02}
   Phase to Debug
   Motion Vector Intensity=Multiply(A=@n44, B=ScalarParameter() {defaultValue=-0.0025, group=1. TextureValues, parameterName="Motion Vector Intensity (S)"})
   Double Apply Motion Vectors
   Custom UV Input
   Custom UV2 Input
   Use Custom UVs
  ) {materialFunction=/Engine/Functions/Engine_MaterialFunctions02/Texturing/FlipBook_MotionVectors}
  NamedRerouteDeclaration(None=@n45.Result) {name=EmissiveTexture, nodeColor=<rgba 0 1 0.2734 1>}
  NamedRerouteDeclaration(None=@n45.Alpha) {name=OpacityTexture, nodeColor=<rgba 1 0 0.4219 1>}
  n46 = ScalarParameter() {defaultValue=16, parameterName=Frame Rate}
  n47 = ScalarParameter() {defaultValue=24, parameterName=Frame Rate Max}
  n48 = VertexColor()
  n49 = MaterialFunctionCall!(
   Texture=TextureObject() {texture=&T_Smoke_Thick_8x8_02}
   UVs
   SubImages=Constant2Vector() {g=8, r=8}
   Frame=Multiply(A=Time(), B=LinearInterpolate(A=@n46, B=@n47, Alpha=@n48.G))
  ) {materialFunction=/Engine/Functions/Engine_MaterialFunctions03/Texturing/SubUV_Function}
  NamedRerouteDeclaration(None=@n49.Alpha) {name=OpacityTexture_Low, nodeColor=<rgba 1 0 0.4219 1>}
  n50 = LinearInterpolate(
   A=ScalarParameter() {defaultValue=1, group=1. TextureValues, parameterName=AnimationSpeed}
   B=ScalarParameter() {defaultValue=1, group=1. TextureValues, parameterName=AnimationSpeedMax}
   Alpha=VertexColor().G
  )
  n51 = Frac(None=Multiply(A=Time(), B=@n50) {constB=0.1})
  n52 = ScalarParameter() {defaultValue=8, group=1. TextureValues, parameterName="Number of Columns X (S)"}
  n53 = ScalarParameter() {defaultValue=8, group=1. TextureValues, parameterName="Number of Rows Y (S)"}
  n54 = TextureObject()
  n55 = Multiply(A=@n50, B=ScalarParameter() {defaultValue=-0.0025, group=1. TextureValues, parameterName="Motion Vector Intensity (S)"})
  n56 = MaterialFunctionCall(
   "Animation  Phase (0-1)"=@n51
   Number of Columns X=@n52
   Number of Rows Y=@n53
   Texture=TextureObject()
   Motion Vector Texture=@n54
   Motion Vector Intensity=@n55
  ) {materialFunction=&FlipBook_MotionVectors}
  NamedRerouteDeclaration(None=ComponentMask(None=@n56.Result) {r=true}) {name=MotionVector_01_R, nodeColor=<rgba 1 0.5432 0.5104 1>}
  NamedRerouteDeclaration(None=ComponentMask(None=@n56.Result) {g=true}) {name=MotionVector_01_G, nodeColor=<rgba 0.0608 0.3229 0 1>}
  NamedRerouteDeclaration(None=ComponentMask(None=@n56.Result) {b=true}) {name=MotionVector_01_B, nodeColor=<rgba 0 0.1124 0.3229 1>}
  n57 = MaterialFunctionCall(
   "Animation  Phase (0-1)"=@n51
   Number of Columns X=@n52
   Number of Rows Y=@n53
   Texture=TextureObject()
   Motion Vector Texture=@n54
   Motion Vector Intensity=@n55
  ) {materialFunction=&FlipBook_MotionVectors}
  NamedRerouteDeclaration(None=ComponentMask(None=@n57.Result) {r=true}) {name=MotionVector_02_R, nodeColor=<rgba 1 0.0623 0.0822 1>}
  NamedRerouteDeclaration(None=ComponentMask(None=@n57.Result) {g=true}) {name=MotionVector_02_G, nodeColor=<rgba 0.1579 0.8385 0 1>}
  NamedRerouteDeclaration(None=ComponentMask(None=@n57.Result) {b=true}) {name=MotionVector_02_B, nodeColor=<rgba 0 0.3426 0.9844 1>}
  NamedRerouteDeclaration(None=@n49.RGB) {name=EmissiveTexture_Low, nodeColor=<rgba 0 1 0.2734 1>}
  NamedRerouteDeclaration(None=@n56.Alpha) {name=MotionVector_01_A, nodeColor=<rgba 1 1 1 1>}
  NamedRerouteUsage()
  NamedRerouteUsage()
  ObjectBounds()
  ObjectPositionWS()
  ObjectRadius()
  ParticlePositionWS()
  ShadingPathSwitch(Default=ShadingPathSwitch(Default=QualitySwitch(Default=@n24, Low=@n24), Mobile=@n24) {desc=Migrated from FeatureLevelSwitch}, Forward=@n24)
  n58 = MaterialFunctionCall() {materialFunction=&CameraDirectionVector}
  n59 = AtmosphericLightVector()
  n60 = Transform(None=Normalize(VectorInput=Subtract(A=@n58.Result, B=@n59))) {transformSourceType=TRANSFORMSOURCE_World, transformType=TRANSFORM_Tangent}
  n61 = ComponentMask(None=@n60) {r=true}
  n62 = NamedRerouteUsage()
  n63 = NamedRerouteUsage()
  n64 = ComponentMask(None=@n60) {g=true}
  n65 = NamedRerouteUsage()
  n66 = NamedRerouteUsage()
  n67 = Add(
   A=Multiply(A=Multiply(A=@n61, B=@n61), B=If(A=@n61, "A > B"=@n62, "A < B"=@n63))
   B=Multiply(A=Multiply(A=@n64, B=@n64), B=If(A=@n64, "A > B"=@n65, "A < B"=@n66))
  )
  n68 = ComponentMask(None=@n60) {b=true}
  n69 = Multiply(A=Multiply(A=@n68, B=@n68), B=If(A=@n68, "A > B"=NamedRerouteUsage(), "A < B"=NamedRerouteUsage()))
  StaticSwitchParameter(
   True=Multiply(A=Add(A=@n67, B=@n69), B=ScalarParameter() {defaultValue=1, parameterName=EmissiveMult})
   False=Multiply(A=NamedRerouteUsage(), B=ScalarParameter() {defaultValue=1, parameterName=EmissiveMult})
  ) {desc=asd, parameterName=Use6PointLightTextures?}
  Transform(None=@n59) {transformSourceType=TRANSFORMSOURCE_World, transformType=TRANSFORM_Tangent}
  n70 = CollectionParameter() {
   collection=&FortniteMaterialParameters
   parameterName=SunAndMoonModelDirectionalVector
  }
  Transform(None=@n70) {transformSourceType=TRANSFORMSOURCE_World, transformType=TRANSFORM_Tangent}
  n71 = CameraVectorWS()
  Transform(None=DotProduct(A=ComponentMask(None=@n59) {b=true, g=true, r=true}, B=OneMinus(None=@n71))) {transformSourceType=TRANSFORMSOURCE_World, transformType=TRANSFORM_Tangent}
  n72 = CollectionParameter() {
   collection=&FortniteMaterialParameters
   parameterName=SunlightDirectionalVector
  }
  MaterialExpressionDotProduct_1 = DotProduct(A=ComponentMask(None=@n72) {b=true, g=true, r=true}, B=CameraVectorWS())
  MaterialExpressionScalarParameter_11 = ScalarParameter() {defaultValue=0.4, parameterName=Anisotropy}
  n73 = Custom(CosTheta=@MaterialExpressionDotProduct_1, k=@MaterialExpressionScalarParameter_11) {
   code="\tfloat Inner = (1 + k * CosTheta);\r\n\treturn (1 - k * k) / (4 * PI * Inner * Inner);\r\n\r\n"
   description=Schlick Phase
   inputs [
    {inputName=CosTheta, input {expression=&self:MaterialExpressionDotProduct_1}}
    {inputName=k, input {expression=&self:MaterialExpressionScalarParameter_11}}
   ]
  }
  n74 = VertexNormalWS()
  n75 = Fresnel(
   ExponentIn=ScalarParameter() {defaultValue=4, parameterName=LightThroughCloudsFresnelXP}
   Normal=Multiply(A=LinearInterpolate(A=@n74), B=TwoSidedSign())
  ) {baseReflectFraction=0.02}
  n76 = OneMinus(None=@n75)
  n77 = ScalarParameter() {parameterName=InvertLightThroughCloudsFresnel}
  VertexInterpolator(VS=Multiply(A=Multiply(A=@n73), B=LinearInterpolate(A=@n75, B=@n76, Alpha=@n77)))
 ]
}
