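# Lit translucent dust/smoke material that animates billboarded puffs entirely in shader, built for instanced meshes.
# Time plus a UV1 phase loops each sprite along a direction while sine-scaling it, with optional landscape height conforming via runtime virtual texture and near-camera shrinking.
# An 8x8 flipbook supplies color and alpha, with opacity shaped by distance, depth, time-of-day and erosion.
# A Schlick phase and Fresnel term computes per-vertex sunlight scattering.
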Material /Game/Effects/Fort_Effects/Materials/Dust/M_Dust_NoTick_Smoke_Lit
graph {
 properties {
  bGenerateSphericalParticleNormals=true
  bUsedWithInstancedStaticMeshes=true
  blendMode=BLEND_Translucent
  pixelDepthOffsetMode=PDOM_Legacy
  translucencyLightingMode=TLM_VolumetricDirectional
  twoSided=true
  usedShadingModels=Default Lit
 }
 nodes [
  n1 = WorldPosition()
  n2 = Multiply(A=ScalarParameter() {defaultValue=0.25, group=3. WPO, parameterName=MinSizeMult}, B=VertexColor().G)
  n3 = ScalarParameter() {defaultValue=500, group=3. WPO, parameterName=MaxSize}
  n4 = Divide(A=Time(), B=ScalarParameter() {defaultValue=5, parameterName=LoopDuration}) {constB=5}
  n5 = TextureCoordinate() {coordinateIndex=1}
  n6 = Multiply(A=Subtract(A=ComponentMask(None=@n5) {g=true}) {constB=0.5}) {constB=-2}
  n7 = StaticSwitchParameter(True=Add(A=Frac(None=@n4), B=@n6), False=@n6) {defaultValue=true, parameterName=Continuous Up and down scaling?}
  n8 = Add(A=@n7) {constB=-0.25}
  n9 = LinearInterpolate(Alpha=ConstantBiasScale(None=Sine(None=@n8)))
  n10 = Multiply(
   A=LinearInterpolate(A=Multiply(A=@n2, B=@n3), B=@n3, Alpha=Power(Base=@n9) {constExponent=1})
   B=MaterialFunctionCall(Distance=ScalarParameter() {defaultValue=500, group=3. WPO, parameterName=NearCamShrinkDist}) {materialFunction=&Near_Camera_Fade}.Result
  )
  n11 = Multiply(A=ObjectBounds(), B=ScalarParameter() {defaultValue=1, parameterName=ObjectRadius}) {constB=0.5}
  n12 = Multiply(A=@n11) {constB=-1}
  n13 = Frac(None=@n7)
  n14 = Multiply(
   A=Multiply(A=LinearInterpolate(A=@n12, B=@n11, Alpha=@n13), B=ScalarParameter() {defaultValue=1, group=WPO, parameterName=OverallSpeed})
   B=VectorParameter() {parameterName=OverallDirection, defaultValue=<rgba 1 0 0 1>}.RGB
  )
  n15 = WorldPosition()
  n16 = ComponentMask(None=ObjectRadius()) {b=true}
  n17 = Add(
   A=Subtract(
    A=RuntimeVirtualTextureSample!(UVs, World Position=@n15.XYZ) {materialType=WorldHeight, virtualTexture=/Game/Athena/Apollo/Maps/Landscape/VT/RVT_Apollo_Landscape_Height}.WorldHeight
    B=ComponentMask(None=@n15.XYZ) {b=true}
   )
   B=@n16
  )
  n18 = Multiply(A=@n16, B=ScalarParameter() {defaultValue=1, group=3. WPO, parameterName=LandScapeOffsetHeight})
  n19 = ComponentMask(None=Multiply(A=VectorParameter() {parameterName=ConformMult, defaultValue=<rgba 0 0 1 1>}.RGB, B=Subtract(A=@n17, B=@n18))) {b=true}
  n20 = Constant()
  n21 = AppendVector(
   A=ComponentMask(None=@n14) {g=true, r=true}
   B=Add(A=ComponentMask(None=@n14) {b=true}, B=StaticSwitchParameter(True=@n19, False=@n20) {parameterName=UseSpawnonLandscape?})
  )
  n22 = Transform(None=@n21)
  Add(
   A=Add(
    A=MaterialFunctionCall!(Center Location=@n1.XYZ, 0-1 UVs, XY scale=@n10, OverridePivot, "Normals (Optional)") {materialFunction=&Sprite}.World position offset
    B=StaticSwitchParameter(True=@n22, False=@n21) {parameterName=UseWorldSpaceMovement?}
   )
   B=MaterialFunctionCall(Offset Amount=ScalarParameter() {group=3. WPO, parameterName=CameraOffsetAmount}) {materialFunction=/Engine/Functions/Engine_MaterialFunctions02/WorldPositionOffset/CameraOffset}.Clamped Camera Offset
  )
  n23 = NamedRerouteUsage()
  n24 = FeatureLevelSwitch(Default=QualitySwitch(Default=@n23, Low=@n23), ES2=@n23, ES3_1=@n23)
  MaterialFunctionCall(Emissive=Multiply(A=ScalarParameter() {defaultValue=1, parameterName=EmissiveMult}, B=ShadingPathSwitch(Default=@n24, Forward=@n23))) {materialFunction=&MF_FX_EmissiveAdjustments}
  Multiply(
   A=LinearInterpolate(A=ScalarParameter() {parameterName=BrightnessMin}, Alpha=@n23)
   B=VectorParameter() {parameterName=MainColor, defaultValue=<rgba 1 1 1 1>}.RGB
  )
  n25 = ScalarParameter() {defaultValue=15000, group=2. Opacity, parameterName="NearFadeLength (S)"}
  n26 = ScalarParameter() {defaultValue=10000, group=2. Opacity, parameterName="NearFadeOffset (S)"}
  n27 = ScalarParameter() {defaultValue=15000, group=2. Opacity, parameterName="FarFadeLength (S)"}
  n28 = ScalarParameter() {defaultValue=75000, group=2. Opacity, parameterName="FarFadeOffset (S)"}
  n29 = NamedRerouteUsage()
  n30 = FeatureLevelSwitch(Default=QualitySwitch(Default=@n29, Low=@n29), ES2=@n29, ES3_1=@n29)
  n31 = VertexColor()
  n32 = Multiply(
   A=VertexInterpolator(VS=MaterialFunctionCall!(
     NearFadeLength=@n25
     NearFadeOffset=@n26
     NearFadeCurve
     NearFadeStartingScale
     NearFadeEndingScale
     FarFadeLength=@n27
     FarFadeOffset=@n28
     FarFadeCurve
     FarFadeEndingScale
     Opacity Near Distance Bias
     Opacity Far Distance Bias
    ) {materialFunction=/Game/Effects/Fort_Effects/Materials/_Functions/MF_SmoothLod}.Opacity).PS
   B=Multiply(A=ShadingPathSwitch(Default=@n30, Forward=@n29), B=LinearInterpolate(Alpha=@n31.B) {constA=0.5})
  )
  n33 = DepthFade(FadeDistance=ScalarParameter() {defaultValue=150, group=2. Opacity, parameterName=Opacity_Depth_FadeDistance})
  n34 = FeatureLevelSwitch(Default=QualitySwitch(Default=Multiply(A=@n32, B=@n33), Low=@n32), ES3_1=@n32)
  n35 = VertexInterpolator(VS=MaterialFunctionCall(
    Value_Morning=ScalarParameter() {defaultValue=1, group=2. Opacity, parameterName="Value_Morning (S)"}
    Value_Day=ScalarParameter() {defaultValue=1, group=2. Opacity, parameterName="Value_Day (S)"}
    Value_Evening=ScalarParameter() {defaultValue=1, group=2. Opacity, parameterName="Value_Evening (S)"}
    Value_Night=ScalarParameter() {defaultValue=1, group=2. Opacity, parameterName="Value_Night (S)"}
   ) {materialFunction=&Scalar_Time_of_Day}.Result)
  n36 = Subtract(A=Multiply(A=@n9, B=Multiply(A=@n34, B=@n35.PS)), B=ScalarParameter() {parameterName=Erosion})
  n37 = ScalarParameter() {parameterName=Erosion_Softness}
  Multiply(A=Saturate(None=Divide(A=@n36, B=@n37)), B=ScalarParameter() {defaultValue=1, group=2. Opacity, parameterName=OpacityMult})
  n38 = ScalarParameter() {defaultValue=16, parameterName=Frame Rate}
  n39 = ScalarParameter() {defaultValue=24, parameterName=Frame Rate Max}
  n40 = VertexColor()
  n41 = MaterialFunctionCall!(
   Texture=TextureObject() {texture=/Game/Effects/Fort_Effects/Textures/Smoke/T_Smoke_Thick_8x8_02}
   UVs
   SubImages=Constant2Vector() {g=8, r=8}
   Frame=Multiply(A=Time(), B=LinearInterpolate(A=@n38, B=@n39, Alpha=@n40.G))
  ) {materialFunction=/Engine/Functions/Engine_MaterialFunctions03/Texturing/SubUV_Function}
  NamedRerouteDeclaration(None=@n41.Alpha) {name=OpacityTexture_Low, nodeColor=<rgba 1 0 0.4219 1>}
  NamedRerouteDeclaration(None=@n41.RGB) {name=EmissiveTexture_Low, nodeColor=<rgba 0 1 0.2734 1>}
  n42 = CollectionParameter() {
   collection=&FortniteMaterialParameters
   parameterName=SunlightDirectionalVector
  }
  MaterialExpressionDotProduct_1 = DotProduct(A=ComponentMask(None=@n42) {b=true, g=true, r=true}, B=CameraVectorWS())
  MaterialExpressionScalarParameter_11 = ScalarParameter() {defaultValue=0.4, parameterName=Anisotropy}
  n43 = Custom!(CosTheta=@MaterialExpressionDotProduct_1, k=@MaterialExpressionScalarParameter_11) {
   code="\tfloat Inner = (1 + k * CosTheta);\r\n\treturn (1 - k * k) / (4 * PI * Inner * Inner);\r\n\r\n"
   description=Schlick Phase
   inputs [
    {inputName=CosTheta, input {expression=&self:MaterialExpressionDotProduct_1}}
    {inputName=k, input {expression=&self:MaterialExpressionScalarParameter_11}}
   ]
  }
  n44 = VertexNormalWS()
  n45 = Fresnel(
   ExponentIn=ScalarParameter() {defaultValue=4, parameterName=LightThroughCloudsFresnelXP}
   Normal=Multiply(A=LinearInterpolate(A=@n44), B=TwoSidedSign())
  ) {baseReflectFraction=0.02}
  n46 = OneMinus(None=@n45)
  n47 = ScalarParameter() {parameterName=InvertLightThroughCloudsFresnel}
  VertexInterpolator(VS=Multiply(A=Multiply(A=@n43), B=LinearInterpolate(A=@n45, B=@n46, Alpha=@n47)))
  VertexNormalWS()
 ]
}
