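# Unlit translucent material for drifting dust motes, tinted by time-of-day morning, day, evening and night colors.
# World position offset combines sprite expansion with panning world-space curl noise, so movement happens in the material.
# Near-camera fade drives sprite scale, smaller up close, and blends color and opacity, fading nearby motes out.
# A soft mask texture shapes opacity, with vertex red varying brightness and alpha and distance cull fade applied.
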Material /Game/Effects/Fort_Effects/Materials/Dust/M_Motes_Creative
graph {
 properties {
  allowTranslucentCustomDepthWrites=true
  blendMode=BLEND_Translucent
  opacityMaskClipValue=0.1
  pixelDepthOffsetMode=PDOM_Legacy
  shadingModel=MSM_Unlit
  translucencyLightingMode=TLM_VolumetricPerVertexNonDirectional
  usedShadingModels=Unlit
 }
 nodes [
  n1 = Add(
   A=Divide(A=WorldPosition().XYZ, B=ScalarParameter() {defaultValue=100, parameterName=CurlWorldSpaceTileScale})
   B=Multiply(A=Time() {bOverride_Period=true, period=20}, B=ScalarParameter() {defaultValue=0.05, parameterName=CurlPanSpeedXYZ})
  )
  n2 = MaterialFunctionCall(Distance=ScalarParameter() {defaultValue=800, parameterName=SoftenDistance}) {materialFunction=&Near_Camera_Fade}
  n3 = LinearInterpolate(
   A=ScalarParameter() {defaultValue=1, parameterName=SpriteScaleNear}
   B=ScalarParameter() {defaultValue=10, parameterName=SpriteScaleFar}
   Alpha=@n2.Result
  )
  Add(
   A=Multiply(A=ScalarParameter() {defaultValue=20, parameterName=CurlAmountScale}, B=VectorNoise(World Position=@n1) {noiseFunction=VNF_CurlALU})
   B=MaterialFunctionCall!(Center Location=WorldPosition().XYZ, 0-1 UVs, XY scale=AppendVector(A=@n3, B=@n3), OverridePivot, "Normals (Optional)") {materialFunction=&Sprite}.World position offset
  )
  n4 = MaterialFunctionCall(
   Color_Morning=VectorParameter() {parameterName=Color_Morning, defaultValue=<rgba 0.3906 0.2775 0.3486 0>}.RGB
   Color_Day=VectorParameter() {parameterName=Color_Day, defaultValue=<rgba 0.358 0.3906 0.1831 0>}.RGB
   Color_Evening=VectorParameter() {parameterName=Color_Evening, defaultValue=<rgba 0.3906 0.2189 0.1078 0>}.RGB
   Color_Night=VectorParameter() {parameterName=Color_Night, defaultValue=<rgba 0.1116 0.2865 0.2065 0>}.RGB
  ) {materialFunction=&Color_Time_of_Day}
  LinearInterpolate(A=Multiply(A=@n4.Result) {constB=0.4}, B=Multiply(A=VertexColor().R, B=@n4.Result), Alpha=@n2.Result)
  n5 = TextureSampleParameter2D!(UVs=Multiply(A=TextureCoordinate(), B=ScalarParameter() {defaultValue=1, parameterName=TileUVs}), Level, Apply View MipBias) {mipValueMode=TMVM_MipLevel, parameterName=MoteTexture, texture=/Game/Global/Textures/Default/Masks_and_Gradients/T_New_Soft_Flair}
  n6 = Multiply(A=ScalarParameter() {defaultValue=0.45, parameterName=Opacity}, B=VertexColor().R)
  Multiply(A=Multiply(A=@n5.G, B=LinearInterpolate(B=@n6, Alpha=@n2.Result)), B=DistanceCullFade())
 ]
}
