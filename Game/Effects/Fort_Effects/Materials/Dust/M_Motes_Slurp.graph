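# Additive unlit material for slurp dust motes, drawn as animated, wobbly glowing bubbles.
# World position offset expands quads into sprites that pulse in size, shrink near the camera, drift and loop along a direction, phased by UV1.
# Panned noise distorts a radial mask feeding streaks, phase-shifted sine RGB iridescence, reflection highlights and rim.
# A two-color gradient tints the oversaturated result (negative desaturation via "Hilarity"), while depth fade softens opacity.
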Material /Game/Effects/Fort_Effects/Materials/Dust/M_Motes_Slurp
graph {
 properties {blendMode=BLEND_Additive, pixelDepthOffsetMode=PDOM_Legacy, shadingModel=MSM_Unlit, usedShadingModels=Unlit}
 nodes [
  n1 = ScalarParameter() {defaultValue=0.25, parameterName=MinSizeMult}
  n2 = VertexColor()
  n3 = ScalarParameter() {defaultValue=500, parameterName=maxSize}
  n4 = Time() {bOverride_Period=true, period=3}
  n5 = Time() {bOverride_Period=true, period=4}
  n6 = ComponentMask(None=TextureCoordinate() {coordinateIndex=1}) {g=true}
  n7 = Add(
   A=StaticSwitchParameter(True=Divide(A=@n4) {constB=3}, False=Divide(A=@n5) {constB=4}) {defaultValue=true, parameterName=UseFastSprites}
   B=Multiply(A=Subtract(A=@n6) {constB=0.5}) {constB=-2}
  )
  n8 = ConstantBiasScale(None=Sine(None=Add(A=@n7) {constB=-0.25}))
  n9 = LinearInterpolate(A=Multiply(A=Multiply(A=@n1, B=@n2.G), B=@n3), B=@n3, Alpha=Power(Base=LinearInterpolate(Alpha=@n8)) {constExponent=0.6})
  n10 = MaterialFunctionCall(Distance=ScalarParameter() {defaultValue=500, parameterName=NearCamShrinkDist}) {materialFunction=&Near_Camera_Fade}
  n11 = Multiply(A=Time(), B=ScalarParameter() {defaultValue=0.001, parameterName=VarianceSpeed}) {constB=0.05}
  n12 = ComponentMask(None=TextureCoordinate() {coordinateIndex=1}) {r=true}
  n13 = Add(A=Multiply(A=Time(), B=ScalarParameter() {defaultValue=0.4, parameterName=DriftSpeed}), B=Sine(None=Add(A=@n11, B=@n12)) {period=0.17})
  n14 = AppendVector(A=AppendVector(A=Sine(None=@n13) {period=0.56}, B=Sine(None=@n13) {period=0.7}), B=Sine(None=@n13) {period=0.82})
  n15 = LinearInterpolate(
   A=ScalarParameter() {defaultValue=1, parameterName=MinHorizontalDrift}
   B=ScalarParameter() {defaultValue=1, parameterName=MaxHorizontalDrift}
   Alpha=Sine(None=Add(A=@n11, B=@n12)) {period=0.14}
  )
  n16 = Multiply(A=ObjectRadius(), B=Constant() {r=0.5}) {constB=0.5}
  n17 = Multiply(
   A=LinearInterpolate(A=Multiply(A=@n16) {constB=-1}, B=@n16, Alpha=Frac(None=@n7))
   B=ScalarParameter() {defaultValue=1, parameterName=OverallSpeed}
  )
  n18 = VectorParameter() {parameterName=OverallDirection, defaultValue=<rgba 1 0 0 1>}
  Add(
   A=MaterialFunctionCall!(Center Location=WorldPosition().XYZ, 0-1 UVs, XY scale=Multiply(A=@n9, B=@n10.Result), OverridePivot, "Normals (Optional)") {materialFunction=&Sprite}.World position offset
   B=Add(A=Multiply(A=@n14, B=@n15), B=Multiply(A=@n17, B=@n18.RGB))
  )
  n19 = MaterialFunctionCall(Float2=TextureCoordinate()) {materialFunction=&BreakOutFloat2Components}
  n20 = ScalarParameter() {defaultValue=1, parameterName=Base_U_Tile}
  n21 = ScalarParameter() {defaultValue=1, parameterName=Base_V_Tile}
  n22 = Add(A=TextureCoordinate()) {constB=-0.5}
  n23 = MaterialFunctionCall(Float2=TextureCoordinate()) {materialFunction=&BreakOutFloat2Components}
  n24 = ScalarParameter() {defaultValue=0.3, parameterName=Distort_U_Tile}
  n25 = ScalarParameter() {defaultValue=0.3, parameterName=Distort_V_Tile}
  n26 = Add(A=AppendVector(A=Multiply(A=@n23.R, B=@n24), B=Multiply(A=@n23.G, B=@n25)), B=VertexColor().R)
  n27 = AppendVector(
   A=ScalarParameter() {defaultValue=0.02, parameterName=Distort_U_PanSpeed}
   B=ScalarParameter() {defaultValue=0.15, parameterName=Distort_V_PanSpeed}
  )
  n28 = MaterialFunctionCall() {materialFunction=&LinearGradient}
  n29 = LinearInterpolate(
   B=TextureSample(UVs=Panner(Coordinate=@n26, Speed=@n27) {speedX=0.02, speedY=0.15}) {samplerType=SAMPLERTYPE_Grayscale, texture=&LowResBlurredNoise}.G
   Alpha=LinearInterpolate(A=SquareRoot(None=@n28.UGradient), Alpha=ScalarParameter() {defaultValue=1, parameterName=Lerp_GradientBlendDistortNoise})
  )
  n30 = Add(
   A=AppendVector(A=Multiply(A=@n19.R, B=@n20), B=Multiply(A=@n19.G, B=@n21))
   B=Multiply(A=ScalarParameter() {defaultValue=4.408, parameterName=WibbStrength}, B=Multiply(A=@n22, B=@n29))
  )
  n31 = VectorParameter() {parameterName=SphereMaskPosition, defaultValue=<rgba 0.5 0.5 0 0>}
  n32 = MaterialFunctionCall!(
   UVs=@n30
   CenterPosition=LinearInterpolate(A=ComponentMask(None=@n31.RGB) {g=true, r=true}, B=Constant2Vector() {g=0.5, r=0.5}, Alpha=@n29)
   Radius=ScalarParameter() {defaultValue=0.5, parameterName=SphereMaskRadius}
  ) {materialFunction=/Engine/Functions/Engine_MaterialFunctions01/Gradient/SphereGradient-2D}
  n33 = OneMinus(None=@n32.Result 0-1)
  n34 = Multiply(A=Multiply(A=Add(A=@n30) {constB=-0.5}, B=LinearInterpolate(Alpha=@n33) {constA=0.1})) {constB=2}
  n35 = Panner(Coordinate=@n34) {speedY=0.2}
  n36 = ScalarParameter() {defaultValue=1, parameterName=SphereMaskContrast}
  n37 = Multiply(A=Multiply(A=@n32.Result 0-1, B=Power(Base=@n33, Exp=@n36))) {constB=4}
  n38 = Add(
   A=Multiply(
    A=TextureSample(UVs=@n35) {texture=/Game/Effects/Fort_Effects/Textures/Energy/T_OutpostShockwaveStreaks_02}.G
    B=ScalarParameter() {defaultValue=1.1194, parameterName=Texture Strength}
   )
   B=Multiply(A=ScalarParameter() {defaultValue=1.1837, parameterName=Round Strength}, B=@n37)
  )
  n39 = Time() {period=1}
  n40 = Add(A=Add(A=Multiply(A=@n39) {constB=0.1}, B=Constant() {r=0.0125}), B=@n38)
  n41 = Sine(None=@n40)
  n42 = ScalarParameter() {defaultValue=0.2524, parameterName=SeperationGap}
  n43 = Sine(None=Add(A=@n40, B=@n42) {constB=0.1})
  n44 = Add(A=@n40, B=Multiply(A=@n42) {constB=2}) {constB=0.15}
  n45 = ConstantBiasScale(None=AppendVector(A=AppendVector(A=@n41, B=@n43), B=Sine(None=@n44)))
  n46 = Add(A=@n34) {constB=0.5}
  n47 = TextureSample(UVs=OneMinus(None=@n46)) {texture=/Game/Items/Art_noLOD/Textures/T_Items_MineralPowder_Glass_Reflection}
  n48 = Add(
   A=TextureSample(UVs=@n46) {texture=&T_Items_MineralPowder_Glass_Reflection}.G
   B=Multiply(A=Multiply(A=@n47.G, B=@n45), B=ScalarParameter() {defaultValue=0.5095, parameterName=Secondary Highlight Strength})
  )
  n49 = Add(
   A=Multiply(A=@n45, B=ScalarParameter() {defaultValue=0.2, parameterName=Color Strength})
   B=Multiply(A=Multiply(A=@n48, B=@n37), B=ScalarParameter() {defaultValue=2, parameterName=Highlight Strength})
  )
  n50 = Multiply(A=@n33, B=@n33)
  n51 = Multiply(A=@n50, B=@n50)
  n52 = Multiply(A=Multiply(A=@n51, B=@n51), B=ScalarParameter() {defaultValue=2, parameterName=Edge Strength})
  Multiply(
   A=LinearInterpolate(
    A=VectorParameter() {parameterName=Color A, defaultValue=<rgba 0 0.4488 1 1>}.RGB
    B=VectorParameter() {parameterName=Color B, defaultValue=<rgba 0 1 0.3167 1>}.RGB
    Alpha=@n38
   )
   B=Desaturation(None=Add(A=@n49, B=@n52), Fraction=ScalarParameter() {defaultValue=-1, parameterName=Hilarity})
  )
  Multiply(
   A=DepthFade(Opacity=@n37, FadeDistance=ScalarParameter() {defaultValue=25, parameterName=Depth Fade Distance})
   B=ScalarParameter() {defaultValue=1, parameterName=Alpha}
  )
 ]
}
