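# Unlit translucent material for round embers or dust motes, animated entirely in the shader on mesh cards.
# World position offset turns cards into sprites, scales them by fade or camera distance, and moves them with sine drift plus looping directional travel.
# Second UV channel values offset each card's drift and life cycle, which drives fades.
# Rotating textures or a sphere mask shape opacity and an edge-to-bright HDR color blend, with optional hue shift.
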Material /Game/Effects/Fort_Effects/Materials/Dust/M_NoTick_Embers_Round
graph {
 properties {
  bUsedWithInstancedStaticMeshes=true
  bUsedWithNiagaraMeshParticles=true
  blendMode=BLEND_Translucent
  pixelDepthOffsetMode=PDOM_Legacy
  shadingModel=MSM_Unlit
  twoSided=true
  usedShadingModels=Unlit
 }
 nodes [
  n1 = Multiply(A=Time(), B=ScalarParameter() {defaultValue=0.001, parameterName=VarianceSpeed}) {constB=0.05}
  n2 = ComponentMask(None=TextureCoordinate() {coordinateIndex=1}) {r=true}
  n3 = Add(A=Multiply(A=Time(), B=ScalarParameter() {defaultValue=0.4, parameterName=DriftSpeed}), B=Sine(None=Add(A=@n1, B=@n2)) {period=0.17})
  n4 = AppendVector(A=AppendVector(A=Sine(None=@n3) {period=0.56}, B=Sine(None=@n3) {period=0.7}), B=Sine(None=@n3) {period=0.82})
  n5 = Transform(None=@n4) {transformType=TRANSFORM_Camera}
  n6 = Add(
   A=Multiply(A=Time(), B=ScalarParameter() {defaultValue=0.2, parameterName=RotateSpeed}) {constB=0.1}
   B=ComponentMask(None=Normalize(VectorInput=@n5)) {g=true}
  )
  n7 = StaticSwitchParameter(
   True=Rotator(Coordinate=TextureCoordinate() {uTiling=0.5, vTiling=0.5}, Time=@n6) {centerX=0.25, centerY=0.25, speed=-1.6}
   False=Rotator(Coordinate=TextureCoordinate(), Time=@n6) {speed=-1.6}
  ) {defaultValue=true, parameterName=Use2x2Texture?}
  n8 = TextureSampleParameter2D(UVs=@n7) {parameterName=TextureMiddle, samplerType=SAMPLERTYPE_LinearColor, texture=/Game/Effects/Fort_Effects/Textures/Smoke/SmokeTest2_Soft}
  n9 = ScalarParameter() {defaultValue=1, parameterName="Contrast (S)"}
  n10 = TextureSampleParameter2D(UVs=@n7) {parameterName=Texture, samplerType=SAMPLERTYPE_LinearColor, texture=&SmokeTest2_Soft}
  n11 = SphereMask(
   A=TextureCoordinate()
   B=Constant2Vector() {g=0.5, r=0.5}
   Radius=ScalarParameter() {defaultValue=0.5, parameterName=Radius}
   Hardness=ScalarParameter() {defaultValue=0.25, parameterName=Hardness}
  )
  n12 = StaticSwitchParameter(True=StaticSwitchParameter(True=@n10.G, False=@n10.G) {parameterName=UseGreenChannelForAlpha?}, False=@n11) {defaultValue=true, parameterName=UseTexture?}
  n13 = Divide(A=Time() {bOverride_Period=true, period=3}) {constB=3}
  n14 = ComponentMask(None=TextureCoordinate() {coordinateIndex=1}) {g=true}
  n15 = Add(
   A=StaticSwitchParameter(True=OneMinus(None=@n13), False=@n13) {parameterName=ReverseDirection?}
   B=Multiply(A=Subtract(A=@n14) {constB=0.5}) {constB=-2}
  )
  n16 = ConstantBiasScale(None=Sine(None=Add(A=@n15) {constB=-0.25}))
  n17 = DepthFade(FadeDistance=ScalarParameter() {defaultValue=50, parameterName=FadeDistance})
  n18 = Constant() {r=1}
  n19 = Multiply(
   A=Power(Base=LinearInterpolate(Alpha=@n16)) {constExponent=0.6}
   B=ShadingPathSwitch(Default=QualitySwitch(Default=@n17, Low=@n18), Mobile=@n18)
  )
  Add(
   A=StaticSwitchParameter(
    True=MaterialFunctionCall(In=@n8.G, Contrast=@n9) {materialFunction=&CheapContrast}.Result
    False=Constant()
   ) {defaultValue=true, parameterName=Use2ndTexture?}
   B=Multiply(A=ScalarParameter() {defaultValue=1, parameterName=EdgeTextureMult}, B=Multiply(A=@n12, B=@n19))
  )
  n20 = LinearInterpolate(
   A=ScalarParameter() {defaultValue=1, parameterName=SpriteScaleNear}
   B=ScalarParameter() {defaultValue=25, parameterName=SpriteScaleFar}
   Alpha=MaterialFunctionCall(Distance=ScalarParameter() {defaultValue=800, parameterName=SoftenDistance}) {materialFunction=&Near_Camera_Fade}.Result
  )
  n21 = ScalarParameter() {defaultValue=0.25, parameterName=MinSizeMult}
  n22 = ScalarParameter() {defaultValue=8, parameterName=Size}
  n23 = StaticSwitchParameter(True=AppendVector(A=@n20, B=@n20), False=LinearInterpolate(A=Multiply(A=@n21, B=@n22), B=@n22, Alpha=@n19)) {parameterName=UseScaleByDistance?}
  n24 = MaterialFunctionCall(Distance=ScalarParameter() {defaultValue=500, parameterName=NearCamShrinkDist}) {materialFunction=&Near_Camera_Fade}
  n25 = Sine(None=Add(A=@n1, B=@n2)) {period=0.14}
  n26 = LinearInterpolate(
   A=ScalarParameter() {defaultValue=10, parameterName=MinHorizontalDrift}
   B=ScalarParameter() {defaultValue=100, parameterName=MaxHorizontalDrift}
   Alpha=@n25
  )
  n27 = ScalarParameter() {defaultValue=0.6, parameterName=MinVerticalSpeedMult}
  n28 = Constant() {r=0.5}
  n29 = Multiply(A=ScalarParameter() {defaultValue=0.6, parameterName=MinVerticalSpeedMult_02}, B=@n28)
  n30 = Multiply(
   A=ObjectRadius()
   B=LinearInterpolate(
    A=Multiply(A=@n27, B=@n28)
    B=StaticSwitchParameter(True=@n29, False=@n28) {parameterName=MoveAllDownWards?}
    Alpha=ConstantBiasScale(None=@n25)
   ) {constA=0.5, constB=0.5}
  ) {constB=0.5}
  n31 = Multiply(
   A=LinearInterpolate(A=Multiply(A=@n30) {constB=-1}, B=@n30, Alpha=Frac(None=@n15))
   B=ScalarParameter() {defaultValue=1, parameterName=OverallSpeed}
  )
  n32 = VectorParameter() {parameterName=OverallDirection, defaultValue=<rgba 0 0 1 1>}
  Add(
   A=MaterialFunctionCall!(Center Location=WorldPosition().XYZ, 0-1 UVs, XY scale=Multiply(A=@n23, B=@n24.Result), OverridePivot, "Normals (Optional)") {materialFunction=&Sprite}.World position offset
   B=Add(A=Multiply(A=@n4, B=@n26), B=Multiply(A=@n31, B=@n32.RGB))
  )
  ScalarParameter()
  n33 = ScalarParameter() {parameterName=HueShiftLow}
  n34 = ScalarParameter() {parameterName=HueShift}
  n35 = VertexColor()
  n36 = Multiply(A=@n10.G, B=@n10.G)
  n37 = LinearInterpolate(
   A=VectorParameter() {parameterName=EdgeColor, defaultValue=<rgba 0.776 0 0.0077 0>}.RGB
   B=VectorParameter() {parameterName=BrightColor, defaultValue=<rgba 4 1.089 0.2 0>}.RGB
   Alpha=Multiply(A=StaticSwitchParameter(True=@n36, False=@n11) {defaultValue=true, parameterName=UseTexture?}, B=@n19)
  )
  n38 = MaterialFunctionCall(LDR Input=Clamp(None=@n37), HDR Tint=@n37, HDR Intensity Multiplier=ScalarParameter() {defaultValue=1, parameterName=HDRIntensity}) {materialFunction=&DeriveHDRfromLDR}
  n39 = FeatureLevelSwitch(Default=@n38.RemappedRange, ES2=@n37, SM4_REMOVED=@n38.RemappedRange)
  StaticSwitchParameter(
   True=MaterialFunctionCall(Hue Shift Percentage=LinearInterpolate(A=@n33, B=@n34, Alpha=@n35.G), Texture=@n39) {materialFunction=&HueShift}.Result
   False=@n39
  ) {parameterName=UseHueShift?}
 ]
}
