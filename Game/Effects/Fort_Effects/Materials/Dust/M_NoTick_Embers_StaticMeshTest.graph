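# Translucent, unlit ember/dust material that animates static-mesh quads as camera-facing sprites entirely in the shader, apparently a test variant avoiding ticking particles.
# Time, UV1 and vertex colors drive looping fades, sprite size and WPO drift, with optional rotation and curl noise.
# Radial noise masks shape color and opacity, with pulsing, hue shift, distance and time-of-day options.
# Quality switches simplify HDR, depth fade and motion; SubUV flipbook math is unconnected.
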Material /Game/Effects/Fort_Effects/Materials/Dust/M_NoTick_Embers_StaticMeshTest
graph {
 properties {
  bEnableResponsiveAA=true
  bUseMaterialAttributes=true
  blendMode=BLEND_Translucent
  pixelDepthOffsetMode=PDOM_Legacy
  previewMesh=/Engine/Transient.MaterialEditorMeshComponent_7
  shadingModel=MSM_FromMaterialExpression
  twoSided=true
  usedShadingModels=Unlit | Default Lit
 }
 nodes [
  ActorPositionWS()
  n1 = ScalarParameter() {defaultValue=4, parameterName=XRows}
  n2 = ScalarParameter() {defaultValue=2, parameterName=YRows}
  n3 = VertexColor()
  n4 = LinearInterpolate(A=ScalarParameter() {parameterName=SubUvSpeed}, B=ScalarParameter() {parameterName=SubUvSpeedMax}, Alpha=@n3.B)
  n5 = Multiply(A=VertexColor().G) {constB=4}
  n6 = Frac(None=Multiply(A=Ceil(None=@n5)) {constB=0.25})
  n7 = Fmod(
   A=Multiply(A=Add(A=@n4, B=@n3.R), B=@n6)
   B=Multiply(A=ScalarParameter() {defaultValue=4, parameterName=XRows}, B=ScalarParameter() {defaultValue=2, parameterName=YRows})
  )
  n8 = ScalarParameter() {defaultValue=4, parameterName=XRows}
  n9 = Fmod(A=Floor(None=@n7), B=@n8)
  n10 = Floor(None=Divide(A=@n7, B=@n8))
  n11 = ScalarParameter() {defaultValue=2, parameterName=YRows}
  Add(A=Divide(A=TextureCoordinate(), B=AppendVector(A=@n1, B=@n2)), B=AppendVector(A=Divide(A=@n9, B=@n8), B=Divide(A=@n10, B=@n11)))
  Add(A=@n6, B=Constant() {r=0.1})
  Constant3Vector()
  Constant()
  Multiply(A=@n6) {constB=2}
  n12 = ObjectPositionWS()
  n13 = Subtract(A=Subtract(A=ComponentMask(None=@n12) {r=true}, B=ComponentMask(None=@n12) {g=true}), B=ComponentMask(None=@n12) {b=true, g=true})
  NamedRerouteDeclaration(None=Abs(None=Frac(None=@n13))) {name=RandomBasedOnPosition, nodeColor=<rgba 1 0.3447 0 1>}
  n14 = VectorParameter() {parameterName=CenterSphereMask_Glow_Color, defaultValue=<rgba 5 0 0 1>}
  n15 = SphereMask(A=WorldPosition().XYZ, B=ObjectPositionWS(), Radius=ScalarParameter() {defaultValue=100, parameterName=CenterSphereMask_Radius}) {attenuationRadius=100, hardnessPercent=0}
  n16 = ScalarParameter() {defaultValue=0.5, parameterName=PulseSpeed_Slow}
  n17 = ScalarParameter() {defaultValue=1, parameterName=PulseSpeed_Fast}
  n18 = Add(A=TextureCoordinate() {coordinateIndex=1, vTiling=0.25}, B=VertexColor().G)
  n19 = Frac(None=Add(A=ComponentMask(None=@n18) {g=true}, B=NamedRerouteUsage()))
  n20 = StaticSwitchParameter(
   True=MaterialFunctionCall(Speed=LinearInterpolate(A=@n16, B=@n17, Alpha=@n19) {constA=0.7}) {materialFunction=&DebugTimeSine}.Result
   False=Constant() {r=1}
  ) {defaultValue=true, parameterName=UseColorPulse?}
  n21 = Clamp(None=@n20, Min=ScalarParameter() {parameterName=HueShiftMin, sliderMax=1}, Max=ScalarParameter() {parameterName=HueShiftMax, sliderMax=1}) {maxDefault=0.08, minDefault=-0.08}
  n22 = VectorParameter() {parameterName=EdgeColor}
  n23 = VectorParameter() {parameterName=BrightColor, defaultValue=<rgba 4 1.089 0.2 0>}
  n24 = ComponentMask(None=TextureCoordinate() {coordinateIndex=1, vTiling=0.5}) {g=true}
  n25 = Add(A=Multiply(A=TextureCoordinate() {uTiling=0.5, vTiling=0.5}, B=AppendVector(A=@n24, B=@n24)), B=AppendVector(A=@n24, B=@n24))
  n26 = AppendVector(A=Constant(), B=Constant())
  n27 = VertexColor()
  n28 = Multiply(
   A=Time()
   B=LinearInterpolate(
    A=ScalarParameter() {defaultValue=0.5, parameterName=RotateSpeed_Noise}
    B=ScalarParameter() {defaultValue=1, parameterName=RotateSpeed_Noise_Max}
    Alpha=@n27.B
   )
  ) {constB=0.1}
  n29 = OneMinus(None=@n28)
  n30 = TextureSampleParameter2D(UVs=Rotator(Coordinate=Panner(Coordinate=@n25, Speed=@n26), Time=LinearInterpolate(A=@n28, B=@n29, Alpha=@n27.B)) {centerX=0.25, centerY=0.25, speed=0.02}) {parameterName=NoiseTexture, texture=/Game/Effects/FogVolume/T_TileNoise}
  n31 = AppendVector(A=Constant(), B=Constant() {r=0.5})
  n32 = MaterialFunctionCall(
   UVs=Add(A=TextureCoordinate(), B=Multiply(A=@n30.G, B=@n31))
   CenterPosition=Constant2Vector() {g=0.5, r=0.5}
   Radius=ScalarParameter() {defaultValue=0.35, parameterName=RadiusMask_Noise}
  ) {materialFunction=&RadialGradientExponential}
  n33 = SphereMask(
   A=TextureCoordinate()
   B=Constant2Vector() {g=0.5, r=0.5}
   Radius=ScalarParameter() {defaultValue=0.4, parameterName=Radius}
   Hardness=ScalarParameter() {defaultValue=-1, parameterName=Hardness}
  )
  n34 = Multiply(A=Power(Base=Multiply(A=@n30.G, B=@n32.RadialGradientExponential)) {constExponent=1}, B=@n33)
  n35 = StaticSwitchParameter(True=@n34, False=@n33) {defaultValue=true, parameterName=UseNoiseTexture?}
  n36 = Saturate(None=@n35)
  n37 = Constant() {r=0.04}
  n38 = Multiply(A=Multiply(A=LinearInterpolate(A=@n22.RGB, B=@n23.RGB, Alpha=@n36), B=Subtract(A=@n35, B=@n37)), B=VertexColor().G)
  n39 = MaterialFunctionCall(LDR Input=Clamp(None=@n38), HDR Tint=@n38, HDR Intensity Multiplier=ScalarParameter() {defaultValue=3, parameterName=HDRIntensity}) {materialFunction=&DeriveHDRfromLDR}
  n40 = Multiply(
   A=@n20
   B=MaterialFunctionCall!(
    Low _ PC Low and Med=@n38
    "High _ XB1, PS4"=@n39.RemappedRange
    "Epic _ PC High, XSX, PS5"=@n39.RemappedRange
    --------
    "Mobile Low _ PC Perf, Android"=@n38
    Mobile Med _ Switch=@n38
   ) {materialFunction=/Game/Global/MaterialFunctions/TechArt/MF_Fortnite_QualitySwitch_Vector3}.Result
  )
  n41 = MaterialFunctionCall(Emissive=Multiply(
    A=MaterialFunctionCall(Hue Shift Percentage=@n21, Texture=@n40) {materialFunction=&HueShift}.Result
    B=ScalarParameter() {defaultValue=1, parameterName=Emissive}
   )) {materialFunction=&MF_FX_EmissiveAdjustments}
  n42 = StaticSwitchParameter(True=Multiply(A=Multiply(A=@n14.RGB, B=@n15), B=@n41.Result), False=@n41.Result) {parameterName=UseSphereMaskCenterGlow?}
  n43 = MaterialFunctionCall(
   Value_Morning=ScalarParameter() {defaultValue=1, parameterName="Value_Morning (S)", sliderMax=1}
   Value_Day=ScalarParameter() {defaultValue=1, parameterName="Value_Day (S)", sliderMax=1}
   Value_Evening=ScalarParameter() {defaultValue=1, parameterName="Value_Evening (S)", sliderMax=1}
   Value_Night=ScalarParameter() {defaultValue=1, parameterName="Value_Night (S)", sliderMax=1}
  ) {materialFunction=&Scalar_Time_of_Day}
  n44 = Frac(None=Divide(A=Time(), B=ScalarParameter() {defaultValue=3, group=1.Main values, parameterName=LoopDuration, sliderMax=50, sliderMin=0.1}))
  n45 = ComponentMask(None=TextureCoordinate() {coordinateIndex=1}) {g=true}
  n46 = Add(
   A=StaticSwitchParameter(True=OneMinus(None=@n44), False=@n44) {parameterName=ReverseDirection?}
   B=Multiply(A=Subtract(A=@n45) {constB=0.5}) {constB=-2}
  )
  n47 = Sine(None=Add(A=@n46) {constB=-0.25})
  n48 = Power(Base=LinearInterpolate(Alpha=ConstantBiasScale(None=@n47))) {constExponent=1}
  n49 = Constant() {r=1}
  n50 = DepthFade(FadeDistance=ScalarParameter() {defaultValue=50, parameterName=DepthFadeDistance})
  n51 = Multiply(
   A=Multiply(A=StaticSwitchParameter(True=@n34, False=@n33) {defaultValue=true, parameterName=UseNoiseTexture?}, B=@n48)
   B=MaterialFunctionCall!(
    Low _ PC Low and Med=@n49
    "High _ XB1, PS4"=@n50
    "Epic _ PC High, XSX, PS5"=@n50
    --------
    "Mobile Low _ PC Perf, Android"=@n49
    Mobile Med _ Switch=@n49
   ) {materialFunction=/Game/Global/MaterialFunctions/TechArt/MF_Fortnite_QualitySwitch_Scalar}.Result
  )
  n52 = Multiply(A=@n51, B=NamedRerouteUsage())
  n53 = Multiply(
   A=LinearInterpolate(A=@n51, B=Round(None=@n52), Alpha=ScalarParameter() {parameterName=SpawnRandomBasedOnPosition?})
   B=ScalarParameter() {defaultValue=1, parameterName=OpacityMult}
  )
  n54 = ScalarParameter() {defaultValue=500, group=Scaler, parameterName="NearFadeLength (S)"}
  n55 = ScalarParameter() {defaultValue=500, group=Scaler, parameterName="NearFadeOffset (S)"}
  n56 = ScalarParameter() {defaultValue=1500, group=Scaler, parameterName="FarFadeLength (S)"}
  n57 = ScalarParameter() {defaultValue=1500, group=Scaler, parameterName="FarFadeOffset (S)"}
  n58 = StaticSwitchParameter(
   True=Multiply(
    A=@n53
    B=MaterialFunctionCall!(
     NearFadeLength=@n54
     NearFadeOffset=@n55
     NearFadeCurve
     NearFadeStartingScale
     NearFadeEndingScale
     FarFadeLength=@n56
     FarFadeOffset=@n57
     FarFadeCurve
     FarFadeEndingScale
     Opacity Near Distance Bias
     Opacity Far Distance Bias
    ) {materialFunction=/Game/Effects/Fort_Effects/Materials/_Functions/MF_SmoothLod}.Opacity
   )
   False=@n53
  ) {parameterName=UseSmoothLod?}
  n59 = LinearInterpolate(
   A=ScalarParameter() {defaultValue=1, parameterName=SpriteScaleNear}
   B=ScalarParameter() {defaultValue=25, parameterName=SpriteScaleFar}
   Alpha=MaterialFunctionCall(Distance=ScalarParameter() {defaultValue=800, parameterName=SpriteScale_SmoothScaling}) {materialFunction=&Near_Camera_Fade}.Result
  )
  n60 = ScalarParameter() {defaultValue=8, parameterName=Size}
  n61 = LinearInterpolate(A=Multiply(A=ScalarParameter() {defaultValue=0.25, parameterName=MinSizeMult}, B=@n60), B=@n60, Alpha=@n48)
  n62 = StaticSwitchParameter(True=Multiply(A=AppendVector(A=@n59, B=@n59), B=@n61), False=@n61) {parameterName=UseScaleByDistance?}
  n63 = MaterialFunctionCall(Distance=ScalarParameter() {defaultValue=500, parameterName=NearCamShrinkDist}) {materialFunction=&Near_Camera_Fade}
  n64 = Multiply(A=Time(), B=ScalarParameter() {defaultValue=0.001, parameterName=VarianceSpeed}) {constB=0.05}
  n65 = ComponentMask(None=TextureCoordinate() {coordinateIndex=1}) {r=true}
  n66 = Add(A=Multiply(A=Time(), B=ScalarParameter() {defaultValue=0.4, parameterName=DriftSpeed}), B=Sine(None=Add(A=@n64, B=@n65)) {period=0.17})
  n67 = AppendVector(A=Sine(None=@n66) {period=0.56}, B=Sine(None=@n66) {period=0.7})
  n68 = Sine(None=@n66) {period=0.82}
  n69 = Sine(None=Add(A=@n64, B=@n65)) {period=0.14}
  n70 = LinearInterpolate(
   A=ScalarParameter() {defaultValue=10, parameterName=MinHorizontalDrift}
   B=ScalarParameter() {defaultValue=100, parameterName=MaxHorizontalDrift}
   Alpha=@n69
  )
  n71 = Clamp(None=ScalarParameter() {defaultValue=1, group=WPO Curl, parameterName=CurlMult}) {maxDefault=2}
  n72 = ScalarParameter() {defaultValue=0.6, parameterName=MinVerticalSpeedMult}
  n73 = Constant() {r=0.5}
  n74 = Multiply(A=ScalarParameter() {defaultValue=0.6, parameterName=MinVerticalSpeedMult_02}, B=@n73)
  n75 = Multiply(
   A=ObjectRadius()
   B=LinearInterpolate(
    A=Multiply(A=@n72, B=@n73)
    B=StaticSwitchParameter(True=@n74, False=@n73) {parameterName=MoveAllDownWards?}
    Alpha=ConstantBiasScale(None=@n69)
   ) {constA=0.5, constB=0.5}
  ) {constB=0.5}
  n76 = Multiply(A=@n75) {constB=-1}
  n77 = Frac(None=@n46)
  n78 = Add(
   A=Multiply(A=AppendVector(A=@n67, B=@n68), B=Multiply(A=@n70, B=@n71))
   B=Multiply(A=LinearInterpolate(A=@n76, B=@n75, Alpha=@n77), B=VectorParameter() {parameterName=OverallDirection, defaultValue=<rgba 0 0 1 1>}.RGB)
  )
  n79 = Time()
  n80 = ObjectPositionWS()
  n81 = Add(A=@n79, B=ComponentMask(None=@n80) {b=true})
  n82 = LinearInterpolate(
   A=ScalarParameter() {defaultValue=0.07, group=WPO Rotation, parameterName=RotationSpeed}
   B=ScalarParameter() {defaultValue=0.05, group=WPO Rotation, parameterName=RotationSpeedReverse}
   Alpha=VertexColor().R
  )
  n83 = ScalarParameter() {defaultValue=1, group=WPO Curl, parameterName=CurlMult}
  n84 = RotateAboutAxis(
   NormalizedRotationAxis=Transform(None=VectorParameter() {parameterName=RotationAxis, defaultValue=<rgba 0 0 1 0>}.RGB) {transformSourceType=TRANSFORMSOURCE_Local}
   RotationAngle=Multiply(A=StaticSwitchParameter(True=@n81, False=@n79) {parameterName=UsewithGrassActor?}, B=Multiply(A=@n82, B=@n83))
   PivotPoint=StaticSwitchParameter(True=@n80, False=ActorPositionWS()) {parameterName=UsewithGrassActor?}
   Position=WorldPosition().XYZ
  )
  n85 = Add(A=StaticSwitchParameter(True=Add(A=@n78, B=@n84), False=@n78) {parameterName=UseRotation}, B=Constant())
  n86 = ScalarParameter() {defaultValue=50, group=WPO Curl, parameterName=CurIntensity}
  n87 = ScalarParameter() {defaultValue=100, group=WPO Curl, parameterName=CurIntensityHigh}
  n88 = VertexColor()
  n89 = Multiply(
   A=Multiply(A=ScalarParameter() {defaultValue=1, group=WPO Curl, parameterName=CurlMult}, B=LinearInterpolate(A=@n86, B=@n87, Alpha=@n88.R))
   B=VertexColor().G
  )
  n90 = Subtract(A=WorldPosition().XYZ, B=ActorPositionWS())
  n91 = Multiply(
   A=Time()
   B=Multiply(
    A=ScalarParameter() {defaultValue=1, group=WPO Curl, parameterName=CurlSpeed}
    B=ScalarParameter() {defaultValue=1, group=WPO Curl, parameterName=CurlMult}
   )
  )
  n92 = LinearInterpolate(
   A=ScalarParameter() {defaultValue=1000, group=WPO Curl, parameterName=CurlScale}
   B=ScalarParameter() {defaultValue=500, group=WPO Curl, parameterName=CurlScale_Low}
   Alpha=@n88.R
  )
  n93 = ScalarParameter() {defaultValue=1, group=WPO Curl, parameterName=CurlMult}
  n94 = VectorNoise(World Position=Divide(A=Subtract(A=@n90, B=@n91), B=Divide(A=@n92, B=@n93))) {noiseFunction=VNF_VectorALU}
  n95 = Add(A=@n85, B=StaticSwitchParameter(True=Multiply(A=@n89, B=@n94), False=Constant()) {parameterName=UseCurlNoise?})
  n96 = ShadingPathSwitch(Default=@n95, Deferred=@n95, Mobile=@n85) {desc=Migrated from FeatureLevelSwitch}
  n97 = Add(
   A=MaterialFunctionCall!(Center Location=WorldPosition().XYZ, 0-1 UVs, XY scale=Multiply(A=@n62, B=@n63.Result), OverridePivot, "Normals (Optional)") {materialFunction=&Sprite}.World position offset
   B=QualitySwitch(Default=ShadingPathSwitch(Default=@n96, Forward=@n85), Low=@n85)
  )
  n98 = ShadingModel() {shadingModel=MSM_Unlit}
  n99 = SetMaterialAttributes!(
   MaterialAttributes
   Emissive Color=@n42
   Opacity=StaticSwitchParameter(True=Multiply(A=@n43.Result, B=@n58), False=@n58) {parameterName=UseTOD_Opacity?}
   World Position Offset=@n97
   Shading Model=@n98
  ) {attributeSetTypes [
    #1
    #2
    #3
    #4
   ]}
  n100 = LinearInterpolate(B=VectorParameter() {parameterName=BaseColor, defaultValue=<rgba 1 1 1 1>}.RGB, Alpha=ScalarParameter() {parameterName=UseBaseColor?})
  n101 = SetMaterialAttributes!(
   MaterialAttributes
   Base Color=StaticSwitchParameter(True=Multiply(A=@n100, B=@n41.Result), False=@n100) {parameterName=UseSphereMaskCenterGlow?}
   Emissive Color=@n42
   Opacity=StaticSwitchParameter(True=Multiply(A=@n58, B=@n43.Result), False=@n58) {parameterName=UseTOD_Opacity?}
   World Position Offset=@n97
   Shading Model=@n98
  ) {attributeSetTypes [
    #5
    #1
    #2
    #3
    #4
   ]}
  QualitySwitch(Default=@n99, Low=@n99, High=@n99, Medium=@n99, Epic=@n101)
  ShadingModel()
 ]
}
