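# Translucent, volumetrically lit haze material, usable on Niagara mesh particles, that draws soft drifting dust animated entirely in the shader.
# World position offset turns geometry into camera-distance-scaled sprites, placed by vertex color RGB and swayed by curl noise.
# Opacity multiplies flow-distorted panning noise by a smoke mask, then applies depth, camera-distance and cull fades.
# Color blends contact and base tints, with optional alpha-based or chromatic variants and separate forward-shading emissive.
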Material /Game/Effects/Fort_Effects/Materials/Dust/M_NoTick_Haze
graph {
 properties {
  bUsedWithNiagaraMeshParticles=true
  blendMode=BLEND_Translucent
  opacityMaskClipValue=0.05
  pixelDepthOffsetMode=PDOM_Legacy
  translucencyLightingMode=TLM_VolumetricPerVertexNonDirectional
  usedShadingModels=Default Lit
 }
 nodes [
  n1 = Time() {period=16}
  n2 = ScalarParameter() {defaultValue=0.15, parameterName=Slow}
  n3 = Multiply(A=VertexNormalWS(), B=Add(A=WorldPosition().XYZ, B=Multiply(A=@n1, B=@n2)))
  n4 = VertexColor()
  n5 = ScalarParameter() {defaultValue=3, parameterName=ScaleOffset}
  n6 = Transform(None=Multiply(A=VectorNoise(World Position=@n3) {bTiling=true, noiseFunction=VNF_CurlALU, tileSize=1}, B=Multiply(A=@n4.A, B=@n5))) {transformSourceType=TRANSFORMSOURCE_Local}
  n7 = Constant()
  n8 = ObjectPositionWS()
  n9 = StaticSwitchParameter(True=VertexColor().A, False=Constant() {r=1}) {defaultValue=true, parameterName=Use Vertex Color Alpha Mask for Sprite Scale}
  n10 = MaterialFunctionCall(Distance=ScalarParameter() {defaultValue=484.8872, parameterName=SoftenDistance}) {materialFunction=&Near_Camera_Fade}
  n11 = Multiply(
   A=@n9
   B=LinearInterpolate(
    A=ScalarParameter() {defaultValue=450, parameterName=SpriteScaleNear}
    B=ScalarParameter() {defaultValue=620.4763, parameterName=SpriteScaleFar}
    Alpha=@n10.Result
   )
  )
  n12 = LinearInterpolate(
   A=ScalarParameter() {defaultValue=450, parameterName=SpriteScaleNear_Y}
   B=ScalarParameter() {defaultValue=620.4763, parameterName=SpriteScaleFar_Y}
   Alpha=@n10.Result
  )
  n13 = StaticSwitchParameter(True=AppendVector(A=@n11, B=Multiply(A=@n9, B=@n12)), False=AppendVector(A=@n11, B=@n11)) {parameterName=Allow Different XY Scale for Sprite}
  n14 = VertexColor()
  n15 = Add(
   A=Multiply(A=@n14.R, B=VectorParameter() {parameterName=OffsetX, defaultValue=<rgba 360.3565 0 0 0>}.RGB)
   B=Multiply(A=@n14.G, B=VectorParameter() {parameterName=OffsetY, defaultValue=<rgba -312.1444 0 0 0>}.RGB)
  )
  n16 = Multiply(A=@n14.B, B=VectorParameter() {parameterName=OffsetZ, defaultValue=<rgba 0 0 67.9512 0>}.RGB)
  Add(
   A=Add(
    A=StaticSwitchParameter(True=@n6, False=@n7) {defaultValue=true, parameterName=AnimateSprite}
    B=MaterialFunctionCall!(Center Location=@n8, 0-1 UVs, XY scale=@n13, OverridePivot, "Normals (Optional)") {materialFunction=&Sprite}.World position offset
   )
   B=Transform(None=Add(A=@n15, B=@n16)) {transformSourceType=TRANSFORMSOURCE_Local}
  )
  CameraPositionWS()
  n17 = Add(A=VertexColor().R, B=TextureCoordinate() {coordinateIndex=1})
  n18 = ScalarParameter() {defaultValue=0.25, parameterName=TileSize}
  n19 = Multiply(A=Time(), B=ScalarParameter() {defaultValue=1, parameterName=Panning Scale})
  n20 = ScalarParameter() {defaultValue=1, parameterName=Velocity_Ctrl}
  n21 = Multiply(A=Constant2Vector() {g=-0.028}, B=@n20)
  n22 = TextureSampleParameter2D(UVs=Panner(Coordinate=Multiply(A=@n17, B=@n18), Time=@n19, Speed=@n21) {speedY=-0.028}) {
   desc=Distort the Tiling Mask by this
   parameterName=UVDistort_Normal
   samplerType=SAMPLERTYPE_Normal
   texture=/Game/Global/Textures/Default/Flowmaps/T_SwirlFlow_02_N
  }
  n23 = Multiply(A=ComponentMask(None=@n22.RGB) {g=true, r=true}, B=ScalarParameter() {defaultValue=0.08, parameterName=UVDistort})
  n24 = TextureCoordinate() {coordinateIndex=1, uTiling=0.5, vTiling=0.5}
  n25 = Multiply(A=Constant2Vector() {g=-0.0125}, B=@n20)
  n26 = TextureSampleParameter2D(UVs=Panner(Coordinate=Add(A=@n23, B=@n24), Time=@n19, Speed=@n25) {speedY=-0.0125}) {
   desc=UV1 Tiling Noise Mask
   parameterName=TileAlphaMask
   samplerType=SAMPLERTYPE_LinearGrayscale
   texture=/Game/Global/Textures/Default/Noise/T_SwirlNoise_N_height
  }
  n27 = TextureCoordinate()
  n28 = Multiply(
   A=@n26.G
   B=TextureSampleParameter2D(UVs=Add(A=@n27, B=@n23)) {desc=UV0 AlphaMask, parameterName=AlphaMask, texture=/Game/Effects/Fort_Effects/Textures/T_Smoke_Soft}.G
  )
  n29 = Multiply(A=ScalarParameter() {defaultValue=500, parameterName=FadeDistance}) {constB=0.714}
  n30 = Clamp(None=@n28)
  n31 = DepthFade(Opacity=@n28, FadeDistance=LinearInterpolate(A=Multiply(A=@n29) {constB=0.3}, B=@n29, Alpha=@n30))
  n32 = ScalarParameter() {defaultValue=250, parameterName=NearFadeManipulator}
  n33 = Multiply(
   A=MaterialFunctionCall(Distance=LinearInterpolate(A=@n32, Alpha=@n30) {constA=250, constB=125}) {materialFunction=&Near_Camera_Fade}.Result
   B=SphereMask(A=PixelDepth().R, B=Constant(), Radius=ScalarParameter() {defaultValue=250, parameterName=FarRadius}, Hardness=Constant() {r=0.5})
  )
  n34 = ScalarParameter() {defaultValue=500, parameterName=FadeDistance}
  n35 = Multiply(
   A=StaticSwitchParameter(True=Multiply(A=@n31, B=@n33), False=DepthFade(Opacity=@n30, FadeDistance=@n34)) {defaultValue=true, parameterName=UseNearFade}
   B=ScalarParameter() {defaultValue=0.3714, parameterName=Opacity}
  )
  n36 = DistanceCullFade()
  DepthFade(
   Opacity=StaticSwitchParameter(True=Multiply(A=@n35, B=@n36), False=@n35) {defaultValue=true, parameterName=UseDistanceCullFade}
   FadeDistance=ScalarParameter() {defaultValue=1000, parameterName=DepthFadeDistance}
  )
  n37 = ScalarParameter() {parameterName=Emissive}
  n38 = VectorParameter() {parameterName=ContactColor, defaultValue=<rgba 0.5932 0.9323 0.3842 0>}
  n39 = VectorParameter() {parameterName=BaseColor, defaultValue=<rgba 0.4375 0.3488 0.2871 0>}
  n40 = Multiply(A=Power(Base=@n30, Exp=ScalarParameter() {defaultValue=1, parameterName=ColorContrast Control})) {constB=10}
  n41 = LinearInterpolate(
   A=@n38.RGB
   B=@n39.RGB
   Alpha=DepthFade(FadeDistance=ScalarParameter() {defaultValue=500, parameterName=FadeDistance}) {fadeDistanceDefault=140}
  )
  n42 = Add(A=VertexColor().R, B=TextureCoordinate() {coordinateIndex=1})
  n43 = ScalarParameter() {defaultValue=0.25, parameterName=TileSize_1}
  n44 = Time()
  n45 = ScalarParameter() {defaultValue=1, parameterName=Panning Scale}
  n46 = TextureSampleParameter2D(UVs=Panner(Coordinate=Multiply(A=@n42, B=@n43), Time=Multiply(A=@n44, B=@n45), Speed=@n21) {speedY=-0.028}) {
   desc=Distort the Tiling Mask by this
   parameterName=UVDistort_Normal_1
   samplerType=SAMPLERTYPE_Normal
   texture=&T_SwirlFlow_02_N
  }
  n47 = Multiply(A=ComponentMask(None=@n46.RGB) {g=true, r=true}, B=ScalarParameter() {defaultValue=0.08, parameterName=UVDistort})
  n48 = Add(B=TextureCoordinate() {coordinateIndex=1, uTiling=0.5, vTiling=0.5}) {constA=0.1}
  n49 = TextureSampleParameter2D(UVs=Panner(Coordinate=Add(A=@n47, B=@n48), Time=@n19, Speed=@n25) {speedY=-0.0125}) {
   desc=UV1 Tiling Noise Mask
   parameterName=TileAlphaMask_2
   samplerType=SAMPLERTYPE_LinearGrayscale
   texture=&T_SwirlNoise_N_height
  }
  n50 = Add(B=TextureCoordinate() {coordinateIndex=1, uTiling=0.5, vTiling=0.5}) {constA=-0.1}
  n51 = TextureSampleParameter2D(UVs=Panner(Coordinate=Add(A=@n47, B=@n50), Time=@n19, Speed=@n25) {speedY=-0.0125}) {
   desc=UV1 Tiling Noise Mask
   parameterName=TileAlphaMask_1
   samplerType=SAMPLERTYPE_LinearGrayscale
   texture=&T_SwirlNoise_N_height
  }
  n52 = TextureCoordinate() {coordinateIndex=1, uTiling=0.5, vTiling=0.5}
  n53 = TextureSampleParameter2D(UVs=Panner(Coordinate=Add(A=@n47, B=@n52), Time=@n19, Speed=@n25) {speedY=-0.0125}) {
   desc=UV1 Tiling Noise Mask
   parameterName=TileAlphaMask_3
   samplerType=SAMPLERTYPE_LinearGrayscale
   texture=&T_SwirlNoise_N_height
  }
  n54 = AppendVector(A=AppendVector(A=@n49.R, B=@n51.R), B=@n53.B)
  n55 = StaticSwitchParameter(True=Multiply(A=@n41, B=@n54), False=@n54) {parameterName=Tint Chromatic}
  n56 = StaticSwitchParameter(
   True=LinearInterpolate(A=@n38.RGB, B=@n39.RGB, Alpha=Saturate(None=@n40))
   False=StaticSwitchParameter(True=@n55, False=@n41) {parameterName=UseChromatic}
  ) {parameterName=Color Based on Alpha Range}
  n57 = ScalarParameter() {defaultValue=0.35, parameterName=Forward Render Multiplier}
  MaterialFunctionCall(Emissive=ShadingPathSwitch(Default=Multiply(A=@n56, B=@n37) {constB=0.1}, Forward=Multiply(A=@n56, B=@n57))) {materialFunction=&MF_FX_EmissiveAdjustments}
  ShadingPathSwitch(Default=@n56, Forward=Constant())
  StaticSwitchParameter(True=@n56, False=@n55) {parameterName=Color Based on Alpha Range}
 ]
}
