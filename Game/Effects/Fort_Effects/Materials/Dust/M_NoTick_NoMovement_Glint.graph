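# Masked, two-sided unlit material for glint sprites on instanced meshes, billboarded through world position offset with only zero constants added as movement.
# A noise-modulated sphere mask drives opacity and an edge-to-bright color gradient, with hue shift, HDR remap on higher quality tiers and optional sine pulse.
# Object position and vertex green randomize pulse speed, size and optional spawning.
# Sprites shrink near the camera, with optional distance scaling, LOD fade and time-of-day opacity.
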Material /Game/Effects/Fort_Effects/Materials/Dust/M_NoTick_NoMovement_Glint
graph {
 properties {
  bEnableResponsiveAA=true
  bUseMaterialAttributes=true
  bUsedWithInstancedStaticMeshes=true
  blendMode=BLEND_Masked
  pixelDepthOffsetMode=PDOM_Legacy
  shadingModel=MSM_Unlit
  twoSided=true
  usedShadingModels=Unlit
 }
 nodes [
  n1 = Constant() {r=1}
  n2 = DepthFade(FadeDistance=ScalarParameter() {defaultValue=50, parameterName=DepthFadeDistance})
  MaterialFunctionCall!(
   Low _ PC Low and Med=@n1
   "High _ XB1, PS4"=@n2
   "Epic _ PC High, XSX, PS5"=@n2
   --------
   "Mobile Low _ PC Perf, Android"=@n1
   Mobile Med _ Switch=@n1
  ) {materialFunction=/Game/Global/MaterialFunctions/TechArt/MF_Fortnite_QualitySwitch_Scalar}
  n3 = ObjectPositionWS()
  n4 = Subtract(A=Subtract(A=ComponentMask(None=@n3) {r=true}, B=ComponentMask(None=@n3) {g=true}), B=ComponentMask(None=@n3) {b=true, g=true})
  NamedRerouteDeclaration(None=Abs(None=Frac(None=@n4))) {name=RandomBasedOnPosition, nodeColor=<rgba 1 0.3447 0 1>}
  n5 = TextureSampleParameter2D(UVs=TextureCoordinate()) {parameterName=Glint, texture=/Game/Effects/FogVolume/T_TileNoise}
  n6 = Constant()
  n7 = Constant() {r=0.5}
  n8 = Add(A=TextureCoordinate(), B=Multiply(A=@n5.G, B=AppendVector(A=@n6, B=@n7)))
  n9 = Constant2Vector() {g=0.5, r=0.5}
  n10 = ScalarParameter() {defaultValue=0.35, parameterName=RadiusMask_Noise}
  Power(Base=Multiply(
    A=@n5.G
    B=MaterialFunctionCall(UVs=@n8, CenterPosition=@n9, Radius=@n10) {materialFunction=&RadialGradientExponential}.RadialGradientExponential
   )) {constExponent=1}
  n11 = ScalarParameter() {defaultValue=0.5, parameterName=PulseSpeed_Slow}
  n12 = ScalarParameter() {defaultValue=1, parameterName=PulseSpeed_Fast}
  n13 = Add(A=TextureCoordinate() {coordinateIndex=1, vTiling=0.25}, B=VertexColor().G)
  n14 = Frac(None=Add(A=ComponentMask(None=@n13) {g=true}, B=NamedRerouteUsage()))
  n15 = LinearInterpolate(A=MaterialFunctionCall(Speed=LinearInterpolate(A=@n11, B=@n12, Alpha=@n14) {constA=0.7}) {materialFunction=&DebugTimeSine}.Result) {constA=0.5}
  n16 = Constant() {r=1}
  n17 = VectorParameter() {parameterName=EdgeColor}
  n18 = VectorParameter() {parameterName=BrightColor, defaultValue=<rgba 4 1.089 0.2 0>}
  n19 = SphereMask(
   A=TextureCoordinate()
   B=Constant2Vector() {g=0.5, r=0.5}
   Radius=ScalarParameter() {defaultValue=0.4, parameterName=Radius}
   Hardness=ScalarParameter() {defaultValue=-1, parameterName=Hardness}
  )
  n20 = Multiply(A=@n5.G, B=@n19)
  n21 = StaticSwitchParameter(True=@n20, False=@n19) {defaultValue=true, parameterName=UseNoiseTexture?}
  n22 = Saturate(None=@n21)
  n23 = Constant() {r=0.04}
  n24 = Multiply(A=Multiply(A=LinearInterpolate(A=@n17.RGB, B=@n18.RGB, Alpha=@n22), B=Subtract(A=@n21, B=@n23)), B=VertexColor().G)
  n25 = MaterialFunctionCall(LDR Input=Clamp(None=@n24), HDR Tint=@n24, HDR Intensity Multiplier=ScalarParameter() {defaultValue=3, parameterName=HDRIntensity}) {materialFunction=&DeriveHDRfromLDR}
  n26 = MaterialFunctionCall(
   Hue Shift Percentage=LinearInterpolate(
    A=ScalarParameter() {parameterName=HueShiftMin, sliderMax=1}
    B=ScalarParameter() {parameterName=HueShiftMax, sliderMax=1}
    Alpha=VertexColor().G
   )
   Texture=Multiply(
    A=StaticSwitchParameter(True=@n15, False=@n16) {defaultValue=true, parameterName=UseColorPulse?}
    B=MaterialFunctionCall!(
     Low _ PC Low and Med=@n24
     "High _ XB1, PS4"=@n25.RemappedRange
     "Epic _ PC High, XSX, PS5"=@n25.RemappedRange
     --------
     "Mobile Low _ PC Perf, Android"=@n24
     Mobile Med _ Switch=@n24
    ) {materialFunction=/Game/Global/MaterialFunctions/TechArt/MF_Fortnite_QualitySwitch_Vector3}.Result
   )
  ) {materialFunction=&HueShift}
  n27 = MaterialFunctionCall(Emissive=Multiply(A=@n26.Result, B=ScalarParameter() {defaultValue=1, parameterName=Emissive})) {materialFunction=&MF_FX_EmissiveAdjustments}
  n28 = MaterialFunctionCall(
   Value_Morning=ScalarParameter() {defaultValue=1, parameterName="Value_Morning (S)", sliderMax=1}
   Value_Day=ScalarParameter() {defaultValue=1, parameterName="Value_Day (S)", sliderMax=1}
   Value_Evening=ScalarParameter() {defaultValue=1, parameterName="Value_Evening (S)", sliderMax=1}
   Value_Night=ScalarParameter() {defaultValue=1, parameterName="Value_Night (S)", sliderMax=1}
  ) {materialFunction=&Scalar_Time_of_Day}
  n29 = Multiply(
   A=Multiply(A=StaticSwitchParameter(True=@n20, False=@n19) {defaultValue=true, parameterName=UseNoiseTexture?}, B=Constant() {r=1})
   B=Constant() {r=1}
  )
  n30 = Multiply(A=@n29, B=NamedRerouteUsage())
  n31 = Multiply(
   A=LinearInterpolate(A=@n29, B=Round(None=@n30), Alpha=ScalarParameter() {parameterName=SpawnRandomBasedOnPosition?})
   B=ScalarParameter() {defaultValue=1, parameterName=OpacityMult}
  )
  n32 = ScalarParameter() {defaultValue=500, group=Scaler, parameterName="NearFadeLength (S)"}
  n33 = ScalarParameter() {defaultValue=500, group=Scaler, parameterName="NearFadeOffset (S)"}
  n34 = ScalarParameter() {defaultValue=1500, group=Scaler, parameterName="FarFadeLength (S)"}
  n35 = ScalarParameter() {defaultValue=1500, group=Scaler, parameterName="FarFadeOffset (S)"}
  n36 = StaticSwitchParameter(
   True=Multiply(
    A=@n31
    B=MaterialFunctionCall!(
     NearFadeLength=@n32
     NearFadeOffset=@n33
     NearFadeCurve
     NearFadeStartingScale
     NearFadeEndingScale
     FarFadeLength=@n34
     FarFadeOffset=@n35
     FarFadeCurve
     FarFadeEndingScale
     Opacity Near Distance Bias
     Opacity Far Distance Bias
    ) {materialFunction=/Game/Effects/Fort_Effects/Materials/_Functions/MF_SmoothLod}.Opacity
   )
   False=@n31
  ) {parameterName=UseSmoothLod?}
  n37 = StaticSwitchParameter(True=Multiply(A=@n28.Result, B=@n36), False=@n36) {parameterName=UseTOD_Opacity?}
  n38 = LinearInterpolate(
   A=ScalarParameter() {defaultValue=1, parameterName=SpriteScaleNear}
   B=ScalarParameter() {defaultValue=25, parameterName=SpriteScaleFar}
   Alpha=MaterialFunctionCall(Distance=ScalarParameter() {defaultValue=800, parameterName=SpriteScale_SmoothScaling}) {materialFunction=&Near_Camera_Fade}.Result
  )
  n39 = ScalarParameter() {defaultValue=8, parameterName=Size}
  n40 = LinearInterpolate(
   A=Multiply(A=ScalarParameter() {defaultValue=0.25, parameterName=MinSizeMult}, B=@n39)
   B=@n39
   Alpha=Multiply(A=VertexColor().G, B=NamedRerouteUsage())
  )
  n41 = StaticSwitchParameter(True=Multiply(A=AppendVector(A=@n38, B=@n38), B=@n40), False=@n40) {parameterName=UseScaleByDistance?}
  n42 = MaterialFunctionCall(Distance=ScalarParameter() {defaultValue=500, parameterName=NearCamShrinkDist}) {materialFunction=&Near_Camera_Fade}
  n43 = Add(A=Constant(), B=Constant())
  n44 = Add(A=@n43, B=Constant())
  n45 = ShadingPathSwitch(Default=@n44, Deferred=@n44, Mobile=@n43) {desc=Migrated from FeatureLevelSwitch}
  n46 = Add(
   A=MaterialFunctionCall!(Center Location=WorldPosition().XYZ, 0-1 UVs, XY scale=Multiply(A=@n41, B=@n42.Result), OverridePivot, "Normals (Optional)") {materialFunction=&Sprite}.World position offset
   B=QualitySwitch(Default=ShadingPathSwitch(Default=@n45, Forward=@n43), Low=@n43)
  )
  n47 = SetMaterialAttributes!(
   MaterialAttributes
   Emissive Color=@n27.Result
   Opacity=@n37
   World Position Offset=@n46
   Shading Model=ShadingModel() {shadingModel=MSM_Unlit}
   Opacity Mask=@n37
  ) {attributeSetTypes [
    #1
    #2
    #3
    #4
    #5
   ]}
  n48 = SetMaterialAttributes!(
   MaterialAttributes
   Base Color=@n26.Result
   Emissive Color=@n27.Result
   Opacity
   World Position Offset=@n46
   Shading Model=ShadingModel()
   Opacity Mask=StaticSwitchParameter(True=Multiply(A=@n36, B=@n28.Result), False=@n36) {parameterName=UseTOD_Opacity?}
  ) {attributeSetTypes [
    #6
    #1
    #2
    #3
    #4
    #5
   ]}
  QualitySwitch(Default=@n47, Low=@n47, High=@n47, Medium=@n47, Epic=@n48)
 ]
}
