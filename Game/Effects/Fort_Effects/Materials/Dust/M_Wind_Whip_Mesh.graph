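# Translucent, unlit, two-sided material for Niagara mesh particles that draws wind whip dust, tinted by time of day, particle color and a back-face color.
# Two panning swirl-noise samples, tiled and offset by dynamic parameters and particle random, combine into one pattern.
# Depth-fade-driven power plus an Erode parameter erode it, then U-graded contrast sharpens it.
# A UV sphere mask, squared view-angle fade and optional near-camera fade shape opacity.
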Material /Game/Effects/Fort_Effects/Materials/Dust/M_Wind_Whip_Mesh
graph {
 properties {
  bUsedWithNiagaraMeshParticles=true
  blendMode=BLEND_Translucent
  pixelDepthOffsetMode=PDOM_Legacy
  shadingModel=MSM_Unlit
  twoSided=true
  usedShadingModels=Unlit
 }
 nodes [
  n1 = VectorParameter() {parameterName=MornEveColor, defaultValue=<rgba 0.7 0.5615 0.5615 0>}
  n2 = VectorParameter() {parameterName=NightColor, defaultValue=<rgba 0.3561 0.3534 0.4 0>}
  n3 = TwoSidedSign()
  n4 = ParticleColor()
  n5 = Multiply(
   A=LinearInterpolate(A=VectorParameter() {parameterName=BackFaceColor, defaultValue=<rgba 0.4567 0.5047 0.526 0>}.RGB, Alpha=ConstantBiasScale(None=@n3))
   B=@n4.RGB
  )
  n6 = VectorParameter() {parameterName=Tile2, defaultValue=<rgba 1.2 0.25 0 0>}
  n7 = DynamicParameter() {defaultValue=<rgba 0 1 1 1>, paramNames [Erode, PanAdd, Param3, HorizontalTile]}
  n8 = AppendVector(A=@n6.R, B=Multiply(A=@n6.G, B=@n7.HorizontalTile))
  n9 = TextureCoordinate()
  n10 = Add(A=@n7.PanAdd, B=ParticleRandom()) {constB=0}
  n11 = TextureSampleParameter2D(UVs=Panner(Coordinate=Multiply(A=@n8, B=@n9), Time=@n10) {speedX=-0.15, speedY=0.35}) {parameterName=Texture, samplerType=SAMPLERTYPE_LinearGrayscale, texture=/Game/Global/Textures/Default/Noise/T_SwirlNoise_N_height}
  n12 = OneMinus(None=@n11.G)
  n13 = VectorParameter() {parameterName=Tile, defaultValue=<rgba 2 0.7 0 0>}
  n14 = AppendVector(A=@n13.R, B=Multiply(A=@n13.G, B=@n7.HorizontalTile))
  n15 = TextureSampleParameter2D(UVs=Panner(Coordinate=Multiply(A=@n9, B=@n14), Time=@n10) {speedX=-0.1, speedY=0.15}) {parameterName=Texture, samplerType=SAMPLERTYPE_LinearGrayscale, texture=&T_SwirlNoise_N_height}
  n16 = LinearInterpolate(Alpha=@n15.G) {constA=1, constB=0.15}
  n17 = Saturate(None=Multiply(A=Add(A=@n12, B=@n16)) {constB=0.5})
  n18 = ComponentMask(None=TextureCoordinate()) {r=true}
  n19 = OneMinus(None=@n18)
  n20 = ScalarParameter() {defaultValue=80, parameterName=DepthFadeDist}
  n21 = Add(
   A=LinearInterpolate(
    A=ScalarParameter() {defaultValue=2, parameterName=MaxPower}
    B=ScalarParameter() {parameterName=MinPower}
    Alpha=DepthFade(Opacity=@n19, FadeDistance=@n20)
   )
   B=DynamicParameter() {defaultValue=<rgba 0 1 1 1>, paramNames [Erode, PanAdd, Param3, HorizontalTile]}.Erode
  )
  n22 = ScalarParameter() {defaultValue=-1, parameterName=MinContrast}
  n23 = ScalarParameter() {defaultValue=3, parameterName=MaxContrast}
  n24 = Constant() {r=0.5}
  n25 = SphereMask(A=TextureCoordinate(), B=@n24, Radius=@n24, Hardness=ScalarParameter() {defaultValue=0.8, parameterName=SphereMaskHardness})
  n26 = Clamp(
   None=MaterialFunctionCall(In=Power(Base=@n17, Exp=@n21), Contrast=LinearInterpolate(A=@n22, B=@n23, Alpha=@n18) {constA=5, constB=0.4}) {materialFunction=&CheapContrast}.Result
   Max=@n25
  ) {maxDefault=2}
  n27 = Multiply(A=LinearInterpolate(Alpha=@n17) {constA=0.3}, B=@n26)
  MaterialFunctionCall(Emissive=Multiply(
    A=MaterialFunctionCall(Color_Morning=@n1.RGB, Color_Evening=@n1.RGB, Color_Night=@n2.RGB) {materialFunction=&Color_Time_of_Day}.Result
    B=Multiply(A=@n5, B=@n27)
   )) {materialFunction=&MF_FX_EmissiveAdjustments}
  n28 = MaterialFunctionCall("FadeSensitivity (optional)"=ScalarParameter() {defaultValue=1, parameterName=AngleFade}) {materialFunction=&FadeBasedOnViewAngle}
  n29 = Multiply(A=@n28.Result, B=@n28.Result)
  n30 = Multiply(A=@n4.A, B=Multiply(A=@n26, B=Multiply(A=@n25, B=@n29)))
  n31 = MaterialFunctionCall(Distance=ScalarParameter() {defaultValue=200, parameterName=Camera Distance Fade}) {materialFunction=&Near_Camera_Fade}
  StaticSwitchParameter(True=Multiply(A=@n30, B=Multiply(A=@n31.Result, B=@n31.Result)), False=@n30) {parameterName=Use Camera Distance Fade}
  TextureSample() {samplerType=SAMPLERTYPE_Normal, texture=/Game/Global/Textures/Default/Noise/T_SwirlNoise_N}
 ]
}
